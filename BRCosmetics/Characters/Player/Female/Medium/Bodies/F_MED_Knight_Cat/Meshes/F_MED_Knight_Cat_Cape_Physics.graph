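# Physics asset for the F_MED_Knight_Cat cosmetic's shoulder cape, defining bodies and joints for its dynamic bones.
# Capsules on pelvis, spine, clavicles and arms form the body, joined by angularly free constraints.
# Seven cape chains hang from MSTR bodies, three on the left upper arm and four on spine_05, with angularly free anchors.
# Chain links use limited swing and very tight twist; only the right-back panel uses boxes.
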PhysicsAsset /BRCosmetics/Characters/Player/Female/Medium/Bodies/F_MED_Knight_Cat/Meshes/F_MED_Knight_Cat_Cape_Physics
bodies {
 clavicle_l [{shapeType=Capsule, length=10.663, radius=3.5999, shapeName=clavicle_l_capsule, center=<3.3688 0.0001 0>, rotation=<rot 90 -2.5307 -2.5309>}]
 clavicle_r [{shapeType=Capsule, length=16.6415, radius=5.0866, shapeName=clavicle_r_capsule, center=<-4.4234 0.564 0>, rotation=<rot 90 0.0197 0.0194>}]
 dyn_shouldercape_bk_01 [{shapeType=Capsule, length=10.3005, radius=2.6996, shapeName=dyn_cape_bk_01_capsule, center=<6 0 0>, rotation=<rot 90 0 0>}]
 dyn_shouldercape_bk_02 [{shapeType=Capsule, length=13.0681, radius=2.6722, shapeName=dyn_cape_bk_02_capsule, center=<6.3587 0 -0.7772>, rotation=<rot 90 0 0>}]
 dyn_shouldercape_bk_MSTR [{shapeType=Capsule, length=1, radius=1, shapeName=dyn_cape_bk_MSTR_capsule}]
 dyn_shouldercape_l_01 [{shapeType=Capsule, length=13.0828, radius=3.2435, shapeName=dyn_cape_l_01_capsule, center=<4.9328 0 1.2486>, rotation=<rot 0 0 89.9999>}]
 dyn_shouldercape_l_02 [{shapeType=Capsule, length=13.0828, radius=3.2435, shapeName=dyn_cape_l_02_capsule, center=<5.3649 0 1.2486>, rotation=<rot 0 0 89.9999>}]
 dyn_shouldercape_l_03 [{shapeType=Capsule, length=13.0828, radius=3.2435, shapeName=dyn_cape_l_03_capsule, center=<2 0 1.2487>, rotation=<rot 0 0 90>}]
 dyn_shouldercape_l_MSTR [{shapeType=Capsule, length=1, radius=1, shapeName=dyn_cape_l_MSTR_capsule, rotation=<rot 0 -90 0>}]
 dyn_shouldercape_l_bk_01 [{shapeType=Capsule, length=5.9943, radius=2.771, shapeName=dyn_cape_l_bk_01_capsule, center=<2 -3.7997 2.3901>, rotation=<rot 0 10 89.9999>}]
 dyn_shouldercape_l_bk_02 [{
   shapeType=Capsule
   length=9.9905
   radius=3.0095
   shapeName=dyn_cape_l_bk_02_capsule
   center=<0.7543 -2.8778 0.5962>
   rotation=<rot 1.2926 14.9462 90.1684>
  }]
 dyn_shouldercape_l_bk_03 [{shapeType=Capsule, length=14.2405, radius=3.3595, shapeName=dyn_cape_l_bk_03_capsule, rotation=<rot 0.0003 10.0002 89.9978>}]
 dyn_shouldercape_l_bk_MSTR [{shapeType=Capsule, length=1, radius=1, shapeName=dyn_cape_l_bk_MSTR_capsule, rotation=<rot 0 -90 0>}]
 dyn_shouldercape_l_fr_01 [{shapeType=Capsule, length=6.0081, radius=2.7381, shapeName=dyn_cape_l_fr_01_capsule, center=<5.9324 0 0.6742>, rotation=<rot 90 0 0>}]
 dyn_shouldercape_l_fr_02 [{shapeType=Capsule, length=5.8431, radius=2.7381, shapeName=dyn_cape_l_fr_02_capsule, center=<3.8795 0 0.3084>, rotation=<rot 90 0 0>}]
 dyn_shouldercape_l_fr_MSTR [{shapeType=Capsule, length=2, radius=1.5, shapeName=dyn_cape_l_fr_MSTR_capsule, rotation=<rot 0 0 90>}]
 dyn_shouldercape_l_frmid_01 [{shapeType=Capsule, length=8.7408, radius=2, shapeName=dyn_cape_l_frmid_01_capsule, rotation=<rot 0 0 90>}]
 dyn_shouldercape_l_frmid_02 [{shapeType=Capsule, length=8.7408, radius=2, shapeName=dyn_cape_l_frmid_02_capsule, rotation=<rot 0 0 90>}]
 dyn_shouldercape_l_frmid_03 [{shapeType=Capsule, length=8.0342, radius=2.7066, shapeName=dyn_cape_l_frmid_03_capsule, rotation=<rot 0 0 90>}]
 dyn_shouldercape_l_frmid_MSTR [{shapeType=Capsule, length=1, radius=2, shapeName=dyn_cape_l_frmid_MSTR_capsule}]
 dyn_shouldercape_r_bk_01 [{shapeType=Box, extentX=8.5, extentY=6.5, extentZ=2.5, shapeName=dyn_cape_r_bk_01_box, center=<8.7707 -1 1.167>}]
 dyn_shouldercape_r_bk_02 [{shapeType=Box, extentX=15, extentY=10, extentZ=3, shapeName=dyn_cape_r_bk_02_box, center=<10 0 -0.404>}]
 dyn_shouldercape_r_bk_MSTR [{shapeType=Capsule, length=1, radius=1, shapeName=dyn_cape_r_bk_MSTR_capsule}]
 dyn_shouldercape_r_fr_01 [{shapeType=Capsule, length=3.3879, radius=1.2906, shapeName=dyn_cape_r_fr_01_capsule, center=<4.0931 0 -0.0943>, rotation=<rot 90 0 0>}]
 dyn_shouldercape_r_fr_02 [{shapeType=Capsule, length=3.5127, radius=1.5336, shapeName=dyn_cape_r_fr_02_capsule, center=<6.836 0 -0.0943>, rotation=<rot 90 0 0>}]
 dyn_shouldercape_r_fr_MSTR [{shapeType=Capsule, length=2, radius=1.5, shapeName=dyn_cape_r_fr_MSTR_capsule, rotation=<rot 0 0 90>}]
 lowerarm_l [{
   shapeType=Capsule
   length=17.4344
   radius=4.2421
   shapeName=lowerarm_l_capsule
   center=<9.4533 -0.0001 1.1394>
   rotation=<rot 90 -117.4458 -117.4458>
  }]
 lowerarm_r [{shapeType=Capsule, length=22.7027, radius=5, shapeName=lowerarm_r_capsule, center=<-7.6475 0.0001 0>, rotation=<rot 90 0 0>}]
 pelvis [{shapeType=Capsule, length=14.148, radius=11.3518, shapeName=pelvis_capsule, center=<-1.7588 -2.7184 0.0006>, rotation=<rot 0.0017 -90 0.002>}]
 spine_02 [{shapeType=Capsule, length=9.9611, radius=11.4061, shapeName=spine_02_capsule, center=<0 -0.0407 0>}]
 spine_04 [{shapeType=Capsule, length=13.038, radius=10.2232, shapeName=spine_04_capsule, center=<-4.3929 1.5593 0>}]
 spine_05 [
  {shapeType=Capsule, length=20.0405, radius=6.7882, shapeName=spine_05_capsule, center=<-0.1882 1.2147 0>, rotation=<rot 90 0 0.0001>}
  {shapeType=Capsule, length=11.5401, radius=6.2362, shapeName=spine_05_capsule, center=<-2.8184 1.5739 0>}
 ]
 upperarm_l [{
   shapeType=Capsule
   length=32.353
   radius=4.1527
   shapeName=upperarm_l_capsule
   center=<7.7244 -0.7313 -1.364>
   rotation=<rot 85 -179.9999 -179.9999>
  }]
 upperarm_r [{
   shapeType=Capsule
   length=21.5813
   radius=5
   shapeName=upperarm_r_capsule
   center=<-11.8534 0.6627 0.2191>
   rotation=<rot 90 -90.9147 -90.9147>
  }]
}
constraints [
 {
  bone1Name=spine_05
  bone2Name=spine_04
  swing1LimitDegrees=45
  swing1Motion=Free
  swing2LimitDegrees=45
  swing2Motion=Free
  twistLimitDegrees=45
  twistMotion=Free
 }
 {
  bone1Name=spine_04
  bone2Name=spine_02
  swing1LimitDegrees=45
  swing1Motion=Free
  swing2LimitDegrees=45
  swing2Motion=Free
  twistLimitDegrees=45
  twistMotion=Free
 }
 {
  bone1Name=upperarm_l
  bone2Name=clavicle_l
  swing1LimitDegrees=45
  swing1Motion=Free
  swing2LimitDegrees=45
  swing2Motion=Free
  twistLimitDegrees=45
  twistMotion=Free
 }
 {
  bone1Name=lowerarm_l
  bone2Name=upperarm_l
  swing1LimitDegrees=45
  swing1Motion=Free
  swing2LimitDegrees=45
  swing2Motion=Free
  twistLimitDegrees=45
  twistMotion=Free
 }
 {
  bone1Name=lowerarm_r
  bone2Name=upperarm_r
  swing1LimitDegrees=45
  swing1Motion=Free
  swing2LimitDegrees=45
  swing2Motion=Free
  twistLimitDegrees=45
  twistMotion=Free
 }
 {
  bone1Name=upperarm_r
  bone2Name=clavicle_r
  swing1LimitDegrees=45
  swing1Motion=Free
  swing2LimitDegrees=45
  swing2Motion=Free
  twistLimitDegrees=45
  twistMotion=Free
 }
 {
  bone1Name=clavicle_l
  bone2Name=spine_05
  swing1LimitDegrees=45
  swing1Motion=Free
  swing2LimitDegrees=45
  swing2Motion=Free
  twistLimitDegrees=45
  twistMotion=Free
 }
 {
  bone1Name=clavicle_r
  bone2Name=spine_05
  swing1LimitDegrees=45
  swing1Motion=Free
  swing2LimitDegrees=45
  swing2Motion=Free
  twistLimitDegrees=45
  twistMotion=Free
 }
 {
  bone1Name=spine_02
  bone2Name=pelvis
  swing1LimitDegrees=45
  swing1Motion=Free
  swing2LimitDegrees=45
  swing2Motion=Free
  twistLimitDegrees=45
  twistMotion=Free
 }
 {
  bone1Name=dyn_shouldercape_l_MSTR
  bone2Name=upperarm_l
  swing1LimitDegrees=45
  swing1Motion=Free
  swing2LimitDegrees=45
  swing2Motion=Free
  twistLimitDegrees=45
  twistMotion=Free
 }
 {bone1Name=dyn_shouldercape_l_01, bone2Name=dyn_shouldercape_l_MSTR, swing1LimitDegrees=10, swing2LimitDegrees=45, twistLimitDegrees=2}
 {bone1Name=dyn_shouldercape_l_02, bone2Name=dyn_shouldercape_l_01, swing1LimitDegrees=10, swing2LimitDegrees=45, twistLimitDegrees=2}
 {bone1Name=dyn_shouldercape_l_03, bone2Name=dyn_shouldercape_l_02, swing1LimitDegrees=10, swing2LimitDegrees=45, twistLimitDegrees=2}
 {
  bone1Name=dyn_shouldercape_l_bk_MSTR
  bone2Name=upperarm_l
  swing1LimitDegrees=45
  swing1Motion=Free
  swing2LimitDegrees=45
  swing2Motion=Free
  twistLimitDegrees=45
  twistMotion=Free
 }
 {bone1Name=dyn_shouldercape_l_bk_01, bone2Name=dyn_shouldercape_l_bk_MSTR, swing1LimitDegrees=10, swing2LimitDegrees=45, twistLimitDegrees=2}
 {bone1Name=dyn_shouldercape_l_bk_02, bone2Name=dyn_shouldercape_l_bk_01, swing1LimitDegrees=10, swing2LimitDegrees=45, twistLimitDegrees=2}
 {bone1Name=dyn_shouldercape_l_bk_03, bone2Name=dyn_shouldercape_l_bk_02, swing1LimitDegrees=10, swing2LimitDegrees=45, twistLimitDegrees=2}
 {
  bone1Name=dyn_shouldercape_l_frmid_MSTR
  bone2Name=upperarm_l
  swing1LimitDegrees=45
  swing1Motion=Free
  swing2LimitDegrees=45
  swing2Motion=Free
  twistLimitDegrees=45
  twistMotion=Free
 }
 {
  bone1Name=dyn_shouldercape_l_frmid_01
  bone2Name=dyn_shouldercape_l_frmid_MSTR
  swing1LimitDegrees=45
  swing2LimitDegrees=30
  twistLimitDegrees=1
 }
 {
  bone1Name=dyn_shouldercape_l_frmid_02
  bone2Name=dyn_shouldercape_l_frmid_01
  swing1LimitDegrees=45
  swing2LimitDegrees=30
  twistLimitDegrees=1
 }
 {
  bone1Name=dyn_shouldercape_l_frmid_03
  bone2Name=dyn_shouldercape_l_frmid_02
  swing1LimitDegrees=45
  swing2LimitDegrees=30
  twistLimitDegrees=1
 }
 {
  bone1Name=dyn_shouldercape_bk_MSTR
  bone2Name=spine_05
  swing1LimitDegrees=45
  swing1Motion=Free
  swing2LimitDegrees=45
  swing2Motion=Free
  twistLimitDegrees=45
  twistMotion=Free
 }
 {bone1Name=dyn_shouldercape_bk_01, bone2Name=dyn_shouldercape_bk_MSTR, swing1LimitDegrees=20, swing2LimitDegrees=30, twistLimitDegrees=0.5}
 {bone1Name=dyn_shouldercape_bk_02, bone2Name=dyn_shouldercape_bk_01, swing1LimitDegrees=20, swing2LimitDegrees=30, twistLimitDegrees=1}
 {
  bone1Name=dyn_shouldercape_l_fr_MSTR
  bone2Name=spine_05
  swing1LimitDegrees=45
  swing1Motion=Free
  swing2LimitDegrees=45
  swing2Motion=Free
  twistLimitDegrees=45
  twistMotion=Free
 }
 {bone1Name=dyn_shouldercape_l_fr_01, bone2Name=dyn_shouldercape_l_fr_MSTR, swing1LimitDegrees=30, swing2LimitDegrees=20, twistLimitDegrees=2}
 {bone1Name=dyn_shouldercape_l_fr_02, bone2Name=dyn_shouldercape_l_fr_01, swing1LimitDegrees=30, swing2LimitDegrees=20, twistLimitDegrees=2}
 {
  bone1Name=dyn_shouldercape_r_bk_MSTR
  bone2Name=spine_05
  swing1LimitDegrees=45
  swing1Motion=Free
  swing2LimitDegrees=45
  swing2Motion=Free
  twistLimitDegrees=45
  twistMotion=Free
 }
 {bone1Name=dyn_shouldercape_r_bk_01, bone2Name=dyn_shouldercape_r_bk_MSTR, swing1LimitDegrees=2, swing2LimitDegrees=20, twistLimitDegrees=0.5}
 {bone1Name=dyn_shouldercape_r_bk_02, bone2Name=dyn_shouldercape_r_bk_01, swing1LimitDegrees=20, swing2LimitDegrees=30, twistLimitDegrees=1}
 {
  bone1Name=dyn_shouldercape_r_fr_MSTR
  bone2Name=spine_05
  swing1LimitDegrees=45
  swing1Motion=Free
  swing2LimitDegrees=45
  swing2Motion=Free
  twistLimitDegrees=45
  twistMotion=Free
 }
 {bone1Name=dyn_shouldercape_r_fr_01, bone2Name=dyn_shouldercape_r_fr_MSTR, swing1LimitDegrees=15, swing2LimitDegrees=5, twistLimitDegrees=0.5}
 {bone1Name=dyn_shouldercape_r_fr_02, bone2Name=dyn_shouldercape_r_fr_01, swing1LimitDegrees=15, swing2LimitDegrees=5, twistLimitDegrees=0.5}
]
