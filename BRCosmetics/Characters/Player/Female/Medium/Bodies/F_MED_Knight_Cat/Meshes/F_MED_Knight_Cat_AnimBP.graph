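# Post-process style Anim Blueprint for the female Knight Cat body, driving secondary motion on its cape, shoulder cape and blades.
# It copies the attached parent's pose, applies Twist and Sculpt Control Rigs, and caches it for RigidBody and spring-based Anim Dynamics chains.
# Constraints and bone offsets tie cloth and blades to arms and thighs.
# Backpack, skydiving and front-end states select or weight corrective rigs and offsets.
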AnimBlueprint /BRCosmetics/Characters/Player/Female/Medium/Bodies/F_MED_Knight_Cat/Meshes/F_MED_Knight_Cat_AnimBP
target_skeleton=&/F_MED_Knight_Cat_Skeleton
graphs [
 AnimGraph {nodes [
   Anim Dynamics - Bone: dyn_shouldercape_l_bk_01(Alpha=1, ExternalForce="0,0,0") {
    angularDampingOverride=0.9999
    angularSpringConstant=100
    bAngularSpring=true
    bOverrideAngularDamping=true
    bUseSphericalLimits=true
    simulationSpace=RootRelative
    alphaBoolBlend {blendOption=Linear}
    boundBone {boneName=dyn_shouldercape_l_bk_01}
    componentAppliedLinearAccClamp=<3500 3500 3500>
    componentLinearAccScale=<1.2 1.2 1.2>
    componentLinearVelScale=<1.2 1.2 1.2>
    physicsBodyDefinitions [{
      sphereCollisionRadius=0
      boundBone {boneName=dyn_shouldercape_l_bk_01}
      boxExtents=<20 10 10>
      constraintSetup {angularLimitsMax=<1 10 15>, angularLimitsMin=<-1 -30 -15>, angularTarget=<0 0 0>}
      localJointOffset=<2 0 0>
     }]
    relativeSpaceBone {boneName=spine_05}
    sphericalLimits [{limitRadius=4, drivingBone {boneName=upperarm_l}, sphereLocalOffset=<0 0 0>}]
   }
   Anim Dynamics - Bone: dyn_shouldercape_bk_01(Alpha=1, ExternalForce="0,0,0") {
    angularDampingOverride=0.9999
    angularSpringConstant=100
    bAngularSpring=true
    bOverrideAngularDamping=true
    simulationSpace=RootRelative
    alphaBoolBlend {blendOption=Linear}
    boundBone {boneName=dyn_shouldercape_bk_01}
    componentAppliedLinearAccClamp=<3500 3500 3500>
    componentLinearAccScale=<1.2 1.2 1.2>
    componentLinearVelScale=<1.2 1.2 1.2>
    physicsBodyDefinitions [{
      sphereCollisionRadius=0
      boundBone {boneName=dyn_shouldercape_bk_01}
      boxExtents=<20 10 10>
      constraintSetup {angularLimitsMax=<1 10 15>, angularLimitsMin=<-1 -30 -15>, angularTarget=<0 0 0>}
      localJointOffset=<2 0 0>
     }]
    relativeSpaceBone {boneName=spine_05}
    sphericalLimits [{limitRadius=0, drivingBone {boneName=None}, sphereLocalOffset=<0 0 0>}]
   }
   Look At - Bone: dyn_shouldercape_l_fr_01(Alpha=1) {
    alpha=0
    alphaBoolBlend {blendOption=Linear}
    alphaScaleBiasClamp {interpSpeedIncreasing=20}
    boneToModify {boneName=dyn_shouldercape_l_fr_01}
    lookAtTarget {boneReference {boneName=upperarm_l}}
    lookAt_Axis {axis=<1 0>}
    lookUp_Axis {axis {x=0}}
   }
   n1 = "Transform (Modify) Bone - Bone: dyn_cape_bk_MSTR"!(ComponentPose, Alpha=1, Translation="0,0,0", Rotation="54.571,-5.625,-83.939") -> (Pose) {
    rotationMode=BMM_Replace
    rotationSpace=BCS_BoneSpace
    scaleMode=BMM_Additive
    scaleSpace=BCS_BoneSpace
    translationMode=BMM_Additive
    translationSpace=BCS_BoneSpace
    alphaBoolBlend {blendOption=Linear}
    alphaScaleBiasClamp {bInterpResult=true, interpSpeedDecreasing=5, interpSpeedIncreasing=15}
    boneToModify {boneName=dyn_cape_bk_MSTR}
    rotation=<rot 54.571 -5.625 -83.939>
   }
   n2 = "Transform (Modify) Bone - Bone: dyn_cape_r_bk_MSTR"!(ComponentPose=@n1, Alpha=1, Translation="0,0,0", Rotation="48.325,-44.962,-25.551") -> (Pose) {
    rotationMode=BMM_Replace
    rotationSpace=BCS_BoneSpace
    scaleMode=BMM_Additive
    scaleSpace=BCS_BoneSpace
    translationMode=BMM_Additive
    translationSpace=BCS_BoneSpace
    alphaBoolBlend {blendOption=Linear}
    alphaScaleBiasClamp {bInterpResult=true, interpSpeedDecreasing=5, interpSpeedIncreasing=15}
    boneToModify {boneName=dyn_cape_r_bk_MSTR}
    rotation=<rot 48.325 -44.962 -25.5509>
   }
   n3 = "Transform (Modify) Bone - Bone: dyn_cape_l_bk_MSTR"!(ComponentPose=@n2, Alpha=1, Translation="0.000,0.000,4", Rotation="55.000,0.000,-20.000") -> (Pose) {
    rotationMode=BMM_Replace
    rotationSpace=BCS_BoneSpace
    scaleMode=BMM_Additive
    scaleSpace=BCS_BoneSpace
    translationMode=BMM_Additive
    translationSpace=BCS_BoneSpace
    alphaBoolBlend {blendOption=Linear}
    alphaScaleBiasClamp {bInterpResult=true, interpSpeedDecreasing=5, interpSpeedIncreasing=15}
    boneToModify {boneName=dyn_cape_l_bk_MSTR}
    rotation=<rot 55 0 -20>
    translation {z=4}
   }
   "Transform (Modify) Bone - Bone: dyn_cape_l_MSTR"!(ComponentPose=@n3, Alpha=1, Translation="0,0,0", Rotation="10.000,0.000,0.000") -> (Pose) {
    rotationMode=BMM_Replace
    rotationSpace=BCS_BoneSpace
    scaleMode=BMM_Additive
    scaleSpace=BCS_BoneSpace
    translationMode=BMM_Additive
    translationSpace=BCS_BoneSpace
    alphaBoolBlend {blendOption=Linear}
    alphaScaleBiasClamp {bInterpResult=true, interpSpeedDecreasing=5, interpSpeedIncreasing=15}
    boneToModify {boneName=dyn_cape_l_MSTR}
    rotation {pitch=10}
   }
   n4 = Get wearingBackpack() -> (wearingBackpack:double)
   n5 = Look At - Bone: dyn_shouldercape_l_bk_02(Alpha=@n4) {
    alpha=0
    alphaBoolBlend {blendOption=Linear}
    alphaScaleBiasClamp {bInterpResult=true, interpSpeedIncreasing=20}
    boneToModify {boneName=dyn_shouldercape_l_bk_02}
    lookAtLocation=<45 -11.019 21.8251>
    lookAtTarget {boneReference {boneName=upperarm_l}}
    lookAt_Axis {axis=<1 0>}
    lookUp_Axis {axis {x=0}}
   }
   n6 = "Transform (Modify) Bone - Bone: dyn_shouldercape_l_MSTR"(ComponentPose=@n5, Alpha=@n4, Translation="0,0,0", Rotation="5.000,-10,0.000", Scale="1,1,1") {
    rotationMode=BMM_Additive
    rotationSpace=BCS_BoneSpace
    scaleMode=BMM_Additive
    scaleSpace=BCS_BoneSpace
    translationMode=BMM_Additive
    translationSpace=BCS_BoneSpace
    alphaBoolBlend {blendOption=Linear}
    alphaScaleBiasClamp {bInterpResult=true, interpSpeedDecreasing=5, interpSpeedIncreasing=15}
    boneToModify {boneName=dyn_shouldercape_l_MSTR}
    rotation {pitch=5, yaw=-10}
   }
   n7 = "Constraint - dyn_shouldercape_l_03 (P)"(ComponentPose=@n6, Alpha=@n4, ConstraintWeights_0=1) {
    alpha=0
    alphaBoolBlend {blendOption=Linear}
    boneToModify {boneName=dyn_shouldercape_l_03}
    constraintSetup [{targetBone {boneName=lowerarm_l}}]
    constraintWeights [1]
   }
   n8 = "Constraint - dyn_shouldercape_bk_02 (P)"(ComponentPose=@n7, Alpha=@n4, ConstraintWeights_0=1) {
    alpha=0
    alphaBoolBlend {blendOption=Linear}
    boneToModify {boneName=dyn_shouldercape_bk_02}
    constraintSetup [{targetBone {boneName=spine_05}}]
    constraintWeights [1]
   }
   n9 = "Constraint - dyn_shouldercape_r_bk_02 (P)"(ComponentPose=@n8, Alpha=@n4, ConstraintWeights_0=1) {
    alpha=0
    alphaBoolBlend {blendOption=Linear}
    boneToModify {boneName=dyn_shouldercape_r_bk_02}
    constraintSetup [{targetBone {boneName=spine_05}}]
    constraintWeights [1]
   }
   n10 = "Constraint - dyn_shouldercape_r_bk_03 (P)"(ComponentPose=@n9, Alpha=@n4, ConstraintWeights_0=1) {
    alpha=0
    alphaBoolBlend {blendOption=Linear}
    boneToModify {boneName=dyn_shouldercape_r_bk_03}
    constraintSetup [{targetBone {boneName=spine_04}}]
    constraintWeights [1]
   }
   n11 = "Transform (Modify) Bone - Bone: dyn_shouldercape_bk_MSTR"!(ComponentPose=@n10, Alpha=@n4, Translation="0,0,1", Rotation="-8,0.000,0.000") -> (Pose) {
    alpha=0
    rotationMode=BMM_Replace
    rotationSpace=BCS_BoneSpace
    scaleMode=BMM_Additive
    scaleSpace=BCS_BoneSpace
    translationMode=BMM_Additive
    translationSpace=BCS_BoneSpace
    alphaBoolBlend {blendOption=Linear}
    boneToModify {boneName=dyn_shouldercape_bk_MSTR}
    rotation {pitch=-8}
    translation {z=1}
   }
   n12 = "Transform (Modify) Bone - Bone: dyn_shouldercape_r_bk_MSTR"!(ComponentPose=@n11, Alpha=@n4, Translation="0.234,-0.000,1.119", Rotation="-15.000,0.000,0.000") -> (Pose) {
    alpha=0
    rotationMode=BMM_Replace
    rotationSpace=BCS_BoneSpace
    scaleMode=BMM_Additive
    scaleSpace=BCS_BoneSpace
    translationMode=BMM_Additive
    translationSpace=BCS_BoneSpace
    alphaBoolBlend {blendOption=Linear}
    boneToModify {boneName=dyn_shouldercape_r_bk_MSTR}
    rotation {pitch=-15}
    translation {x=0.234, z=1.119}
   }
   "Transform (Modify) Bone - Bone: dyn_shouldercape_l_bk_MSTR"!(ComponentPose=@n12, Alpha=@n4, Translation="-0.000,0.000,2.789", Rotation="0,0,0") -> (Pose) {
    alpha=0
    rotationMode=BMM_Replace
    rotationSpace=BCS_BoneSpace
    scaleMode=BMM_Additive
    scaleSpace=BCS_BoneSpace
    translationMode=BMM_Additive
    translationSpace=BCS_BoneSpace
    alphaBoolBlend {blendOption=Linear}
    boneToModify {boneName=dyn_shouldercape_l_bk_MSTR}
    translation {z=2.789}
   }
   RigidBody(ComponentPose=Local To Component(LocalPose=Use cached pose 'Pose'())) {
    bClampLinearTranslationLimitToRefPose=true
    bOverrideWorldGravity=true
    overridePhysicsAsset=&/F_MED_Knight_Cat_Cape_Physics
    simulationSpace=BaseBoneSpace
    alphaBoolBlend {blendOption=Linear}
    baseBoneRef {boneName=spine_05}
    componentAppliedLinearAccClamp=<3000 5000 5000>
    componentLinearAccScale=<0.2 0.5 0.5>
    componentLinearVelScale=<0.2 0.5 0.5>
    overrideWorldGravity {z=-980}
   }
   n13 = "Transform (Modify) Bone - Bone: dyn_shouldercape_l_MSTR"(
    ComponentPose=Local To Component(LocalPose=Use cached pose 'SavedPose27606'())
    Alpha=1
    Translation="0,0,0"
    Rotation="3.198,-10.707,0.138"
    Scale="1,1,1"
   ) {
    rotationMode=BMM_Additive
    translationMode=BMM_Additive
    alphaBoolBlend {blendOption=Linear}
    boneToModify {boneName=dyn_shouldercape_l_MSTR}
    rotation=<rot 3.198 -10.707 0.138>
   }
   n14 = "Transform (Modify) Bone - Bone: dyn_shouldercape_l_bk_MSTR"!(ComponentPose=@n13, Alpha=1, Translation="0,0,0", Rotation="10.000,0.000,0.000", Scale:Vector="1,1,1") -> (Pose) {
    rotationMode=BMM_Additive
    translationMode=BMM_Additive
    alphaBoolBlend {blendOption=Linear}
    boneToModify {boneName=dyn_shouldercape_l_bk_MSTR}
    rotation {pitch=10}
   }
   n15 = "Transform (Modify) Bone - Bone: dyn_shouldercape_r_bk_MSTR"!(ComponentPose=@n14, Alpha=1, Translation="0,0,0", Rotation="-15,0.000,-0.000", Scale:Vector="1,1,1") -> (Pose) {
    rotationMode=BMM_Additive
    translationMode=BMM_Additive
    alphaBoolBlend {blendOption=Linear}
    boneToModify {boneName=dyn_shouldercape_r_bk_MSTR}
    rotation {pitch=-15}
   }
   n16 = Get wearingBackpack()
   n17 = Control Rig(Source=Use cached pose 'SavedPose27606'(), Alpha=@n16) {
    alphaBoolBlend {blendOption=Linear}
    controlRigAssetReference {blueprintRigClass=&/Knight_Cat_Backpack_FE_CtrlRig.Knight_Cat_Backpack_FE_CtrlRig_C}
   }
   n18 = Blend Poses by bool(BlendPose_0=Component To Local(ComponentPose=@n15), BlendPose_1=@n17, bActiveValue=Get bIsSkyDiving()) {blendType=Linear}
   n19 = Control Rig(Source=Use cached pose 'SavedPose27606'(), Alpha=@n16) {
    alphaBoolBlend {blendOption=Linear}
    controlRigAssetReference {blueprintRigClass=&/Knight_Cat_Backpack_CtrlRig.Knight_Cat_Backpack_CtrlRig_C}
   }
   n20 = Get bIsInFrontEnd()
   n21 = "Constraint - dyn_shouldercape_l_bk_01 (P)"(
    ComponentPose=Local To Component(LocalPose=Blend Poses by bool(BlendPose_0=@n18, BlendPose_1=@n19, bActiveValue=@n20) {blendType=Linear})
    Alpha=1
    ConstraintWeights_0=1
   ) {
    alphaBoolBlend {blendOption=Linear}
    boneToModify {boneName=dyn_shouldercape_l_bk_01}
    constraintSetup [{targetBone {boneName=upperarm_l}}]
    constraintWeights [1]
   }
   n22 = "Constraint - dyn_shouldercape_l_bk_02 (P)"(ComponentPose=@n21, Alpha=1, ConstraintWeights_0=1) {
    alphaBoolBlend {blendOption=Linear}
    boneToModify {boneName=dyn_shouldercape_l_bk_02}
    constraintSetup [{targetBone {boneName=upperarm_l}}]
    constraintWeights [1]
   }
   n23 = "Constraint - dyn_shouldercape_l_bk_03 (P)"(ComponentPose=@n22, Alpha=1, ConstraintWeights_0=1) {
    alphaBoolBlend {blendOption=Linear}
    boneToModify {boneName=dyn_shouldercape_l_bk_03}
    constraintSetup [{targetBone {boneName=upperarm_l}}]
    constraintWeights [1]
   }
   n24 = "Constraint - dyn_shouldercape_l_01 (P)"(ComponentPose=@n23, Alpha=1, ConstraintWeights_0=1) {
    alphaBoolBlend {blendOption=Linear}
    boneToModify {boneName=dyn_shouldercape_l_01}
    constraintSetup [{targetBone {boneName=upperarm_l}}]
    constraintWeights [1]
   }
   n25 = "Constraint - dyn_shouldercape_l_02 (P)"(ComponentPose=@n24, Alpha=1, ConstraintWeights_0=1) {
    alphaBoolBlend {blendOption=Linear}
    boneToModify {boneName=dyn_shouldercape_l_02}
    constraintSetup [{targetBone {boneName=upperarm_l}}]
    constraintWeights [1]
   }
   n26 = "Constraint - dyn_shouldercape_l_03 (P)"(ComponentPose=@n25, Alpha=1, ConstraintWeights_0=1) {
    alphaBoolBlend {blendOption=Linear}
    boneToModify {boneName=dyn_shouldercape_l_03}
    constraintSetup [{targetBone {boneName=upperarm_l}}]
    constraintWeights [1]
   }
   n27 = "Transform (Modify) Bone - Bone: dyn_shouldercape_r_bk_MSTR"!(ComponentPose=@n26, Alpha=1, Translation="-0.2,0,0", Rotation="25,0.000,0.000") -> (Pose) {
    alpha=0
    rotationMode=BMM_Replace
    rotationSpace=BCS_BoneSpace
    scaleMode=BMM_Additive
    scaleSpace=BCS_BoneSpace
    translationMode=BMM_Replace
    translationSpace=BCS_BoneSpace
    alphaBoolBlend {blendOption=Linear}
    alphaScaleBiasClamp {bInterpResult=true, interpSpeedIncreasing=20}
    boneToModify {boneName=dyn_shouldercape_r_bk_MSTR}
    rotation {pitch=25}
    translation {x=-0.2}
   }
   n28 = "Transform (Modify) Bone - Bone: dyn_shouldercape_l_bk_MSTR"!(ComponentPose=@n27, Alpha=1, Translation="-0.2,0,0", Rotation="25,0.000,0.000") -> (Pose) {
    alpha=0
    rotationMode=BMM_Replace
    rotationSpace=BCS_BoneSpace
    scaleMode=BMM_Additive
    scaleSpace=BCS_BoneSpace
    translationMode=BMM_Replace
    translationSpace=BCS_BoneSpace
    alphaBoolBlend {blendOption=Linear}
    alphaScaleBiasClamp {bInterpResult=true, interpSpeedIncreasing=20}
    boneToModify {boneName=dyn_shouldercape_l_bk_MSTR}
    rotation {pitch=25}
    translation {x=-0.2}
   }
   n29 = "Transform (Modify) Bone - Bone: dyn_shouldercape_bk_MSTR"(ComponentPose=@n28, Alpha=1, Translation="-0.2,0,0", Rotation="30,0.000,0.000") {
    alpha=0
    rotationMode=BMM_Replace
    rotationSpace=BCS_BoneSpace
    scaleMode=BMM_Additive
    scaleSpace=BCS_BoneSpace
    translationMode=BMM_Replace
    translationSpace=BCS_BoneSpace
    alphaBoolBlend {blendOption=Linear}
    alphaScaleBiasClamp {bInterpResult=true, interpSpeedIncreasing=20}
    boneToModify {boneName=dyn_shouldercape_bk_MSTR}
    rotation {pitch=30}
    translation {x=-0.2}
   }
   n30 = Control Rig(Source=Component To Local(ComponentPose=@n29), Alpha=Get noBackpack() -> (noBackpack:double)) {
    alphaBoolBlend {blendOption=Linear}
    controlRigAssetReference {blueprintRigClass=&/F_MED_Knight_Cat_CtrlRig_Arm.F_MED_Knight_Cat_CtrlRig_Arm_C}
   }
   n31 = Get bIsInFrontEnd()
   n32 = "Transform (Modify) Bone - Bone: dyn_shouldercape_r_fr_MSTR"!(ComponentPose=Local To Component(LocalPose=@n30), bAlphaBoolEnabled:bool=@n31, Translation="0,0,0", Rotation="0,0,-8", Scale="1,1,1") -> (Pose) {
    alphaInputType=Bool
    rotationMode=BMM_Additive
    alphaBoolBlend {blendOption=Linear}
    boneToModify {boneName=dyn_shouldercape_r_fr_MSTR}
    rotation {roll=-8}
   }
   n33 = "Transform (Modify) Bone - Bone: dyn_shouldercape_r_fr_02"!(ComponentPose=@n32, bAlphaBoolEnabled:bool=@n31, Translation="0,0,0", Rotation="0,0,3", Scale="1,1,1") -> (Pose) {
    alphaInputType=Bool
    rotationMode=BMM_Additive
    alphaBoolBlend {blendOption=Linear}
    boneToModify {boneName=dyn_shouldercape_r_fr_02}
    rotation {roll=3}
   }
   n34 = "Transform (Modify) Bone - Bone: dyn_blade_r_MST"!(ComponentPose=@n33, bAlphaBoolEnabled:bool=@n31, Translation="0,0,0", Rotation="-3.926,20,5.000", Scale="1,1,1") -> (Pose) {
    alphaInputType=Bool
    rotationMode=BMM_Additive
    alphaBoolBlend {blendOption=Linear}
    boneToModify {boneName=dyn_blade_r_MST}
    rotation=<rot -3.926 20 5>
   }
   n35 = "Transform (Modify) Bone - Bone: dyn_shouldercape_r_bk_MSTR"!(
    ComponentPose=@n34
    bAlphaBoolEnabled:bool=Get frontend_noRBAN() -> (frontend_noRBAN:bool)
    Translation="0,0,0"
    Rotation="-13.995,1.556,11.368"
    Scale:Vector="1,1,1"
   ) -> (Pose) {
    alphaInputType=Bool
    rotationMode=BMM_Additive
    translationMode=BMM_Additive
    translationSpace=BCS_BoneSpace
    alphaBoolBlend {blendInTime=0.1, blendOption=Linear, blendOutTime=0.1}
    boneToModify {boneName=dyn_shouldercape_r_bk_MSTR}
    rotation=<rot -13.995 1.556 11.368>
   }
   n36 = "Transform (Modify) Bone - Bone: attach_backpack"!(
    ComponentPose=@n35
    bAlphaBoolEnabled:bool=Get cape_active() -> (cape_active:bool)
    Translation="0.000,-2,0.000"
    Rotation="0,0,0"
    Scale="1,1,1"
   ) -> (Pose) {
    alphaInputType=Bool
    rotationMode=BMM_Additive
    rotationSpace=BCS_BoneSpace
    translationMode=BMM_Additive
    translationSpace=BCS_BoneSpace
    alphaBoolBlend {blendOption=Linear}
    boneToModify {boneName=attach_backpack}
    translation {y=-2}
   }
   Output Pose(Result=Component To Local(ComponentPose=@n36)) {layerGroup=None}
   n37 = Control Rig(Source=Copy Pose From Mesh() {bUseAttachedParent=true}) {
    alphaBoolBlend {blendOption=Linear}
    controlRigAssetReference {blueprintRigClass=&/F_MED_Knight_Cat_CtrlRig_Twist.F_MED_Knight_Cat_CtrlRig_Twist_C}
   }
   n38 = Control Rig(Source=@n37) {
    alphaBoolBlend {blendOption=Linear}
    controlRigAssetReference {blueprintRigClass=&/F_MED_Knight_Cat_CtrlRig_Sculpt.F_MED_Knight_Cat_CtrlRig_Sculpt_C}
   }
   n39 = "Constraint - dyn_blade_r (PPP)"!(
    ComponentPose=Local To Component(LocalPose=@n38)
    Alpha=1
    ConstraintWeights_0=0.255
    ConstraintWeights_1:float=0.595
    ConstraintWeights_2:float=0.15
   ) -> (Pose) {
    alphaBoolBlend {blendOption=Linear}
    boneToModify {boneName=dyn_blade_r}
    constraintSetup [{targetBone {boneName=pelvis}}, {targetBone {boneName=thigh_twist_01_r}}, {targetBone {boneName=thigh_r}}]
    constraintWeights [1,1,1]
   }
   n40 = "Constraint - dyn_blade_l (PPP)"!(ComponentPose=@n39, Alpha=1, ConstraintWeights_0=0.255, ConstraintWeights_1:float=0.595, ConstraintWeights_2:float=0.15) -> (Pose) {
    alphaBoolBlend {blendOption=Linear}
    boneToModify {boneName=dyn_blade_l}
    constraintSetup [{targetBone {boneName=pelvis}}, {targetBone {boneName=thigh_twist_01_l}}, {targetBone {boneName=thigh_l}}]
    constraintWeights [1,1,1]
   }
   Save cached pose 'Pose'(Pose=Component To Local(ComponentPose=@n40)) {cacheName=Pose}
   n41 = RigidBody(ComponentPose=Local To Component(LocalPose=Use cached pose 'Pose'())) {
    bClampLinearTranslationLimitToRefPose=true
    bOverrideWorldGravity=true
    overridePhysicsAsset=&F_MED_Knight_Cat_Cape_Physics
    simulationSpace=BaseBoneSpace
    alphaBoolBlend {blendOption=Linear}
    baseBoneRef {boneName=spine_05}
    componentAppliedLinearAccClamp=<3000 3000 3000>
    componentLinearAccScale=<0.35 0.8 0.8>
    componentLinearVelScale=<0.35 0.8 0.8>
    overrideWorldGravity {z=-980}
   }
   n42 = Anim Dynamics - Bone: dyn_blade_l(ComponentPose=@n41, Alpha=1, ExternalForce="0,0,0") {
    angularDampingOverride=0.9999
    angularSpringConstant=100
    bAngularSpring=true
    bOverrideAngularDamping=true
    simulationSpace=RootRelative
    alphaBoolBlend {blendOption=Linear}
    boundBone {boneName=dyn_blade_l}
    componentAppliedLinearAccClamp=<3500 3500 3500>
    componentLinearAccScale=<1.2 1.2 1.2>
    componentLinearVelScale=<1.2 1.2 1.2>
    physicsBodyDefinitions [{
      sphereCollisionRadius=0
      boundBone {boneName=dyn_blade_l}
      boxExtents=<20 10 10>
      constraintSetup {angularLimitsMax=<1 2 30>, angularLimitsMin=<-1 -1 -30>, angularTarget=<0 0 0>}
      localJointOffset=<2 0 0>
     }]
    relativeSpaceBone {boneName=pelvis}
   }
   n43 = Anim Dynamics - Bone: dyn_blade_r(ComponentPose=@n42, Alpha=1, ExternalForce="0,0,0") {
    angularDampingOverride=0.9999
    angularSpringConstant=100
    bAngularSpring=true
    bOverrideAngularDamping=true
    simulationSpace=RootRelative
    alphaBoolBlend {blendOption=Linear}
    boundBone {boneName=dyn_blade_r}
    componentAppliedLinearAccClamp=<3500 3500 3500>
    componentLinearAccScale=<1.2 1.2 1.2>
    componentLinearVelScale=<1.2 1.2 1.2>
    physicsBodyDefinitions [{
      sphereCollisionRadius=0
      boundBone {boneName=dyn_blade_r}
      boxExtents=<20 10 10>
      constraintSetup {angularLimitsMax=<1 2 30>, angularLimitsMin=<-1 -1 -30>, angularTarget=<0 0 0>}
      localJointOffset=<2 0 0>
     }]
    relativeSpaceBone {boneName=pelvis}
   }
   n44 = "Transform (Modify) Bone - Bone: dyn_blade_l_MSTR"!(ComponentPose=@n43, Alpha=Get thigh_rot_up_l() -> (thigh_rot_up_l:double), Translation="0,0,0", Rotation="-10.000,0.000,0.000") -> (Pose) {
    rotationMode=BMM_Replace
    rotationSpace=BCS_BoneSpace
    scaleMode=BMM_Additive
    scaleSpace=BCS_BoneSpace
    translationMode=BMM_Additive
    translationSpace=BCS_BoneSpace
    alphaBoolBlend {blendOption=Linear}
    alphaScaleBiasClamp {bInterpResult=true, interpSpeedDecreasing=5, interpSpeedIncreasing=15}
    boneToModify {boneName=dyn_blade_l_MSTR}
    rotation {pitch=-10}
   }
   n45 = "Transform (Modify) Bone - Bone: dyn_blade_r_MST"!(ComponentPose=@n44, Alpha:float=Get thigh_rot_up_r() -> (thigh_rot_up_r:double), Translation="0,0,0", Rotation="25.000,0,0") -> (Pose) {
    rotationMode=BMM_Replace
    rotationSpace=BCS_BoneSpace
    scaleMode=BMM_Additive
    scaleSpace=BCS_BoneSpace
    translationMode=BMM_Additive
    translationSpace=BCS_BoneSpace
    alphaBoolBlend {blendOption=Linear}
    alphaScaleBiasClamp {bInterpResult=true, interpSpeedDecreasing=5, interpSpeedIncreasing=15}
    boneToModify {boneName=dyn_blade_r_MST}
    rotation {pitch=25}
   }
   n46 = "Transform (Modify) Bone - Bone: dyn_shouldercape_r_bk_01"!(ComponentPose=@n45, bAlphaBoolEnabled:bool=True, Translation="0.000,-0.001,-1.410", Rotation="4.696,0.004,-0.003", Scale="1,1,1") -> (Pose) {
    alphaInputType=Bool
    rotationMode=BMM_Additive
    rotationSpace=BCS_BoneSpace
    translationMode=BMM_Additive
    translationSpace=BCS_BoneSpace
    boneToModify {boneName=dyn_shouldercape_r_bk_01}
    rotation=<rot 4.696 0.004 -0.003>
    translation {y=-0.001, z=-1.41}
   }
   n47 = "Transform (Modify) Bone - Bone: dyn_shouldercape_r_bk_02"!(ComponentPose=@n46, bAlphaBoolEnabled:bool=True, Translation="-0.001,-1.802,-1.707", Rotation="9.254,0.006,0.003", Scale="1,1,1") -> (Pose) {
    alphaInputType=Bool
    rotationMode=BMM_Additive
    rotationSpace=BCS_BoneSpace
    translationMode=BMM_Additive
    translationSpace=BCS_BoneSpace
    boneToModify {boneName=dyn_shouldercape_r_bk_02}
    rotation=<rot 9.254 0.006 0.003>
    translation=<-0.001 -1.802 -1.707>
   }
   n48 = "Constraint - dyn_shouldercape_l_02 (P)"(ComponentPose=Local To Component(LocalPose=Use cached pose 'Pose'()), Alpha=0.85, ConstraintWeights_0=1) {
    alphaBoolBlend {blendOption=Linear}
    alphaScaleBiasClamp {bInterpResult=true, interpSpeedIncreasing=20}
    boneToModify {boneName=dyn_shouldercape_l_02}
    constraintSetup [{targetBone {boneName=upperarm_l}}]
    constraintWeights [1]
   }
   n49 = "Constraint - dyn_shouldercape_l_bk_02 (P)"(ComponentPose=@n48, Alpha=0.85, ConstraintWeights_0=1) {
    alphaBoolBlend {blendOption=Linear}
    alphaScaleBiasClamp {bInterpResult=true, interpSpeedIncreasing=20}
    boneToModify {boneName=dyn_shouldercape_l_bk_02}
    constraintSetup [{targetBone {boneName=upperarm_l}}]
    constraintWeights [1]
   }
   n50 = "Constraint - dyn_shouldercape_l_frmid_02 (P)"(ComponentPose=@n49, Alpha=0.85, ConstraintWeights_0=1) {
    alphaBoolBlend {blendOption=Linear}
    alphaScaleBiasClamp {bInterpResult=true, interpSpeedIncreasing=20}
    boneToModify {boneName=dyn_shouldercape_l_frmid_02}
    constraintSetup [{targetBone {boneName=upperarm_l}}]
    constraintWeights [1]
   }
   n51 = "Transform (Modify) Bone - Bone: dyn_shouldercape_l_frmid_MSTR"(ComponentPose=@n50, Alpha=1, Translation="0,0,0", Rotation="5,0.000,0.000", Scale="1,1,1") {
    rotationMode=BMM_Additive
    rotationSpace=BCS_BoneSpace
    scaleMode=BMM_Additive
    scaleSpace=BCS_BoneSpace
    translationMode=BMM_Additive
    translationSpace=BCS_BoneSpace
    alphaBoolBlend {blendOption=Linear}
    boneToModify {boneName=dyn_shouldercape_l_frmid_MSTR}
    rotation {pitch=5}
   }
   n52 = "Transform (Modify) Bone - Bone: dyn_shouldercape_l_MSTR"(ComponentPose=@n51, Alpha=1, Translation="0,0,0", Rotation="20.000,0.000,0.000", Scale="1,1,1") {
    rotationMode=BMM_Additive
    rotationSpace=BCS_BoneSpace
    scaleMode=BMM_Additive
    scaleSpace=BCS_BoneSpace
    translationMode=BMM_Additive
    translationSpace=BCS_BoneSpace
    alphaBoolBlend {blendOption=Linear}
    boneToModify {boneName=dyn_shouldercape_l_MSTR}
    rotation {pitch=20}
   }
   n53 = "Transform (Modify) Bone - Bone: dyn_shouldercape_l_bk_MSTR"!(ComponentPose=@n52, Alpha=1, Translation="0,0,0", Rotation="15.000,0.000,0.000", Scale:Vector="1,1,1") -> (Pose) {
    rotationMode=BMM_Additive
    rotationSpace=BCS_BoneSpace
    scaleMode=BMM_Additive
    scaleSpace=BCS_BoneSpace
    translationMode=BMM_Additive
    translationSpace=BCS_BoneSpace
    alphaBoolBlend {blendOption=Linear}
    boneToModify {boneName=dyn_shouldercape_l_bk_MSTR}
    rotation {pitch=15}
   }
   n54 = Get arm_up_l() -> (arm_up_l:double)
   n55 = "Transform (Modify) Bone - Bone: dyn_shouldercape_l_MSTR"!(ComponentPose=@n53, Alpha=@n54, Rotation="-10.000,0.000,0.000") -> (Pose) {
    rotationMode=BMM_Additive
    rotationSpace=BCS_BoneSpace
    scaleSpace=BCS_BoneSpace
    translationMode=BMM_Additive
    translationSpace=BCS_BoneSpace
    alphaBoolBlend {blendOption=Linear}
    boneToModify {boneName=dyn_shouldercape_l_MSTR}
    rotation {pitch=-10}
   }
   n56 = "Transform (Modify) Bone - Bone: dyn_shouldercape_l_bk_MSTR"!(ComponentPose=@n55, Alpha=@n54, Rotation="5.885,-9.250,-6.202") -> (Pose) {
    rotationMode=BMM_Additive
    rotationSpace=BCS_BoneSpace
    scaleSpace=BCS_BoneSpace
    translationMode=BMM_Additive
    translationSpace=BCS_BoneSpace
    alphaBoolBlend {blendOption=Linear}
    boneToModify {boneName=dyn_shouldercape_l_bk_MSTR}
    rotation=<rot 5.885 -9.25 -6.202>
   }
   n57 = Anim Dynamics - Bone: dyn_shouldercape_l_fr_01(ComponentPose=@n56, Alpha=1, ExternalForce="0,0,0") {
    angularDampingOverride=0.9999
    angularSpringConstant=100
    bAngularSpring=true
    bOverrideAngularDamping=true
    simulationSpace=RootRelative
    alphaBoolBlend {blendOption=Linear}
    boundBone {boneName=dyn_shouldercape_l_fr_01}
    componentAppliedLinearAccClamp=<3500 3500 3500>
    componentLinearAccScale=<1.2 1.2 1.2>
    componentLinearVelScale=<1.2 1.2 1.2>
    physicsBodyDefinitions [{
      sphereCollisionRadius=0
      boundBone {boneName=dyn_shouldercape_l_fr_01}
      boxExtents=<20 10 10>
      constraintSetup {angularLimitsMax=<1 10 15>, angularLimitsMin=<-1 -30 -15>, angularTarget=<0 0 0>}
      localJointOffset=<2 0 0>
     }]
    relativeSpaceBone {boneName=spine_05}
   }
   n58 = Anim Dynamics - Bone: dyn_shouldercape_l_fr_02(ComponentPose=@n57, Alpha=1, ExternalForce="0,0,0") {
    angularDampingOverride=0.9999
    angularSpringConstant=100
    bAngularSpring=true
    bOverrideAngularDamping=true
    simulationSpace=RootRelative
    alphaBoolBlend {blendOption=Linear}
    boundBone {boneName=dyn_shouldercape_l_fr_02}
    componentAppliedLinearAccClamp=<3500 3500 3500>
    componentLinearAccScale=<1.2 1.2 1.2>
    componentLinearVelScale=<1.2 1.2 1.2>
    physicsBodyDefinitions [{
      sphereCollisionRadius=0
      boundBone {boneName=dyn_shouldercape_l_fr_02}
      boxExtents=<20 10 10>
      constraintSetup {angularLimitsMax=<1 10 15>, angularLimitsMin=<-1 -30 -15>, angularTarget=<0 0 0>}
      localJointOffset=<2 0 0>
     }]
    relativeSpaceBone {boneName=spine_05}
   }
   n59 = Anim Dynamics - Bone: dyn_shouldercape_l_fr_03(ComponentPose=@n58, Alpha=1, ExternalForce="0,0,0") {
    angularDampingOverride=0.9999
    angularSpringConstant=100
    bAngularSpring=true
    bOverrideAngularDamping=true
    simulationSpace=RootRelative
    alphaBoolBlend {blendOption=Linear}
    boundBone {boneName=dyn_shouldercape_l_fr_03}
    componentAppliedLinearAccClamp=<3500 3500 3500>
    componentLinearAccScale=<1.2 1.2 1.2>
    componentLinearVelScale=<1.2 1.2 1.2>
    physicsBodyDefinitions [{
      sphereCollisionRadius=0
      boundBone {boneName=dyn_shouldercape_l_fr_03}
      boxExtents=<20 10 10>
      constraintSetup {angularLimitsMax=<1 5 15>, angularLimitsMin=<-1 -30 -15>, angularTarget=<0 0 0>}
      localJointOffset=<2 0 0>
     }]
    relativeSpaceBone {boneName=spine_05}
   }
   n60 = Anim Dynamics - Bone: dyn_shouldercape_l_01(ComponentPose=@n59, Alpha=1, ExternalForce="0,0,0") {
    angularDampingOverride=0.9999
    angularSpringConstant=100
    bAngularSpring=true
    bOverrideAngularDamping=true
    bUseSphericalLimits=true
    simulationSpace=RootRelative
    alphaBoolBlend {blendOption=Linear}
    boundBone {boneName=dyn_shouldercape_l_01}
    componentAppliedLinearAccClamp=<3500 3500 3500>
    componentLinearAccScale=<1.2 1.2 1.2>
    componentLinearVelScale=<1.2 1.2 1.2>
    physicsBodyDefinitions [{
      sphereCollisionRadius=0
      boundBone {boneName=dyn_shouldercape_l_01}
      boxExtents=<20 10 10>
      constraintSetup {angularLimitsMax=<1 5 15>, angularLimitsMin=<-1 -30 -15>, angularTarget=<0 0 0>}
      localJointOffset=<2 0 0>
     }]
    relativeSpaceBone {boneName=spine_05}
    sphericalLimits [{limitRadius=4, drivingBone {boneName=upperarm_l}, sphereLocalOffset=<0 0 0>}]
   }
   n61 = Anim Dynamics - Bone: dyn_shouldercape_l_02(ComponentPose=@n60, Alpha=1, ExternalForce="0,0,0") {
    angularDampingOverride=0.9999
    angularSpringConstant=100
    bAngularSpring=true
    bOverrideAngularDamping=true
    bUseSphericalLimits=true
    simulationSpace=RootRelative
    alphaBoolBlend {blendOption=Linear}
    boundBone {boneName=dyn_shouldercape_l_02}
    componentAppliedLinearAccClamp=<3500 3500 3500>
    componentLinearAccScale=<1.2 1.2 1.2>
    componentLinearVelScale=<1.2 1.2 1.2>
    physicsBodyDefinitions [{
      sphereCollisionRadius=0
      boundBone {boneName=dyn_shouldercape_l_02}
      boxExtents=<20 10 10>
      constraintSetup {angularLimitsMax=<1 5 15>, angularLimitsMin=<-1 -30 -15>, angularTarget=<0 0 0>}
      localJointOffset=<2 0 0>
     }]
    relativeSpaceBone {boneName=spine_05}
    sphericalLimits [{limitRadius=4, drivingBone {boneName=upperarm_l}, sphereLocalOffset=<6 0 0>}]
   }
   n62 = Anim Dynamics - Bone: dyn_shouldercape_l_03(ComponentPose=@n61, Alpha=1, ExternalForce="0,0,0") {
    angularDampingOverride=0.9999
    angularSpringConstant=100
    bAngularSpring=true
    bOverrideAngularDamping=true
    bUseSphericalLimits=true
    simulationSpace=RootRelative
    alphaBoolBlend {blendOption=Linear}
    boundBone {boneName=dyn_shouldercape_l_03}
    componentAppliedLinearAccClamp=<3500 3500 3500>
    componentLinearAccScale=<1.2 1.2 1.2>
    componentLinearVelScale=<1.2 1.2 1.2>
    physicsBodyDefinitions [{
      sphereCollisionRadius=0
      boundBone {boneName=dyn_shouldercape_l_03}
      boxExtents=<20 10 10>
      constraintSetup {angularLimitsMax=<1 5 15>, angularLimitsMin=<-1 -30 -15>, angularTarget=<0 0 0>}
      localJointOffset=<2 0 0>
     }]
    relativeSpaceBone {boneName=spine_05}
    sphericalLimits [{limitRadius=5, drivingBone {boneName=upperarm_l}, sphereLocalOffset=<15 0 0>}]
   }
   n63 = Anim Dynamics - Bone: dyn_shouldercape_l_bk_02(ComponentPose=@n62, Alpha=1, ExternalForce="0,0,0") {
    angularDampingOverride=0.9999
    angularSpringConstant=100
    bAngularSpring=true
    bOverrideAngularDamping=true
    bUseSphericalLimits=true
    simulationSpace=RootRelative
    alphaBoolBlend {blendOption=Linear}
    boundBone {boneName=dyn_shouldercape_l_bk_02}
    componentAppliedLinearAccClamp=<3500 3500 3500>
    componentLinearAccScale=<1.2 1.2 1.2>
    componentLinearVelScale=<1.2 1.2 1.2>
    physicsBodyDefinitions [{
      sphereCollisionRadius=0
      boundBone {boneName=dyn_shouldercape_l_bk_02}
      boxExtents=<20 10 10>
      constraintSetup {angularLimitsMax=<1 10 15>, angularLimitsMin=<-1 -30 -15>, angularTarget=<0 0 0>}
      localJointOffset=<2 0 0>
     }]
    relativeSpaceBone {boneName=spine_05}
    sphericalLimits [{limitRadius=4, drivingBone {boneName=upperarm_l}, sphereLocalOffset=<6 0 0>}]
   }
   n64 = Anim Dynamics - Bone: dyn_shouldercape_l_bk_03(ComponentPose=@n63, Alpha=1, ExternalForce="0,0,0") {
    angularDampingOverride=0.9999
    angularSpringConstant=100
    bAngularSpring=true
    bOverrideAngularDamping=true
    bUseSphericalLimits=true
    simulationSpace=RootRelative
    alphaBoolBlend {blendOption=Linear}
    boundBone {boneName=dyn_shouldercape_l_bk_03}
    componentAppliedLinearAccClamp=<3500 3500 3500>
    componentLinearAccScale=<1.2 1.2 1.2>
    componentLinearVelScale=<1.2 1.2 1.2>
    physicsBodyDefinitions [{
      sphereCollisionRadius=0
      boundBone {boneName=dyn_shouldercape_l_bk_03}
      boxExtents=<20 10 10>
      constraintSetup {angularLimitsMax=<1 5 15>, angularLimitsMin=<-1 -30 -15>, angularTarget=<0 0 0>}
      localJointOffset=<2 0 0>
     }]
    relativeSpaceBone {boneName=spine_05}
    sphericalLimits [{limitRadius=5, drivingBone {boneName=upperarm_l}, sphereLocalOffset=<15 0 0>}]
   }
   n65 = Anim Dynamics - Bone: dyn_shouldercape_bk_02(ComponentPose=@n64, Alpha=1, ExternalForce="0,0,0") {
    angularDampingOverride=0.9999
    angularSpringConstant=100
    bAngularSpring=true
    bOverrideAngularDamping=true
    simulationSpace=RootRelative
    alphaBoolBlend {blendOption=Linear}
    boundBone {boneName=dyn_shouldercape_bk_02}
    componentAppliedLinearAccClamp=<3500 3500 3500>
    componentLinearAccScale=<1.2 1.2 1.2>
    componentLinearVelScale=<1.2 1.2 1.2>
    physicsBodyDefinitions [{
      sphereCollisionRadius=0
      boundBone {boneName=dyn_shouldercape_bk_02}
      boxExtents=<20 10 10>
      constraintSetup {angularLimitsMax=<1 5 15>, angularLimitsMin=<-1 -30 -15>, angularTarget=<0 0 0>}
      localJointOffset=<2 0 0>
     }]
    relativeSpaceBone {boneName=spine_05}
    sphericalLimits [{limitRadius=0, drivingBone {boneName=None}, sphereLocalOffset=<0 0 0>}]
   }
   n66 = Anim Dynamics - Bone: dyn_shouldercape_bk_03(ComponentPose=@n65, Alpha=1, ExternalForce="0,0,0") {
    angularDampingOverride=0.9999
    angularSpringConstant=100
    bAngularSpring=true
    bOverrideAngularDamping=true
    simulationSpace=RootRelative
    alphaBoolBlend {blendOption=Linear}
    boundBone {boneName=dyn_shouldercape_bk_03}
    componentAppliedLinearAccClamp=<3500 3500 3500>
    componentLinearAccScale=<1.2 1.2 1.2>
    componentLinearVelScale=<1.2 1.2 1.2>
    physicsBodyDefinitions [{
      sphereCollisionRadius=0
      boundBone {boneName=dyn_shouldercape_bk_03}
      boxExtents=<20 10 10>
      constraintSetup {angularLimitsMax=<1 5 15>, angularLimitsMin=<-1 -30 -15>, angularTarget=<0 0 0>}
      localJointOffset=<2 0 0>
     }]
    relativeSpaceBone {boneName=spine_05}
    sphericalLimits [{limitRadius=0, drivingBone {boneName=None}, sphereLocalOffset=<0 0 0>}]
   }
   n67 = Anim Dynamics - Bone: dyn_shouldercape_r_bk_01(ComponentPose=@n66, Alpha=1, ExternalForce="0,0,0") {
    angularDampingOverride=0.9999
    angularSpringConstant=100
    bAngularSpring=true
    bOverrideAngularDamping=true
    simulationSpace=RootRelative
    alphaBoolBlend {blendOption=Linear}
    boundBone {boneName=dyn_shouldercape_r_bk_01}
    componentAppliedLinearAccClamp=<3500 3500 3500>
    componentLinearAccScale=<1.2 1.2 1.2>
    componentLinearVelScale=<1.2 1.2 1.2>
    physicsBodyDefinitions [{
      sphereCollisionRadius=0
      boundBone {boneName=dyn_shouldercape_r_bk_01}
      boxExtents=<20 10 10>
      constraintSetup {angularLimitsMax=<1 10 15>, angularLimitsMin=<-1 -30 -15>, angularTarget=<0 0 0>}
      localJointOffset=<2 0 0>
     }]
    relativeSpaceBone {boneName=spine_05}
   }
   n68 = Anim Dynamics - Bone: dyn_shouldercape_r_bk_02(ComponentPose=@n67, Alpha=1, ExternalForce="0,0,0") {
    angularDampingOverride=0.9999
    angularSpringConstant=100
    bAngularSpring=true
    bOverrideAngularDamping=true
    simulationSpace=RootRelative
    alphaBoolBlend {blendOption=Linear}
    boundBone {boneName=dyn_shouldercape_r_bk_02}
    componentAppliedLinearAccClamp=<3500 3500 3500>
    componentLinearAccScale=<1.2 1.2 1.2>
    componentLinearVelScale=<1.2 1.2 1.2>
    physicsBodyDefinitions [{
      sphereCollisionRadius=0
      boundBone {boneName=dyn_shouldercape_r_bk_02}
      boxExtents=<20 10 10>
      constraintSetup {angularLimitsMax=<1 5 15>, angularLimitsMin=<-1 -30 -15>, angularTarget=<0 0 0>}
      localJointOffset=<2 0 0>
     }]
    relativeSpaceBone {boneName=spine_05}
   }
   n69 = Anim Dynamics - Bone: dyn_shouldercape_r_bk_03(ComponentPose=@n68, Alpha=1, ExternalForce="0,0,0") {
    angularDampingOverride=0.9999
    angularSpringConstant=100
    bAngularSpring=true
    bOverrideAngularDamping=true
    simulationSpace=RootRelative
    alphaBoolBlend {blendOption=Linear}
    boundBone {boneName=dyn_shouldercape_r_bk_03}
    componentAppliedLinearAccClamp=<3500 3500 3500>
    componentLinearAccScale=<1.2 1.2 1.2>
    componentLinearVelScale=<1.2 1.2 1.2>
    physicsBodyDefinitions [{
      sphereCollisionRadius=0
      boundBone {boneName=dyn_shouldercape_r_bk_03}
      boxExtents=<20 10 10>
      constraintSetup {angularLimitsMax=<1 5 15>, angularLimitsMin=<-1 -30 -15>, angularTarget=<0 0 0>}
      localJointOffset=<2 0 0>
     }]
    relativeSpaceBone {boneName=spine_05}
   }
   n70 = Anim Dynamics - Bone: dyn_blade_l(ComponentPose=@n69, Alpha=1, ExternalForce="0,0,0") {
    angularDampingOverride=0.9999
    angularSpringConstant=100
    bAngularSpring=true
    bOverrideAngularDamping=true
    simulationSpace=RootRelative
    alphaBoolBlend {blendOption=Linear}
    boundBone {boneName=dyn_blade_l}
    componentAppliedLinearAccClamp=<3500 3500 3500>
    componentLinearAccScale=<1.2 1.2 1.2>
    componentLinearVelScale=<1.2 1.2 1.2>
    physicsBodyDefinitions [{
      sphereCollisionRadius=0
      boundBone {boneName=dyn_blade_l}
      boxExtents=<20 10 10>
      constraintSetup {angularLimitsMax=<1 2 30>, angularLimitsMin=<-1 -1 -30>, angularTarget=<0 0 0>}
      localJointOffset=<2 0 0>
     }]
    relativeSpaceBone {boneName=pelvis}
   }
   n71 = Anim Dynamics - Bone: dyn_blade_r(ComponentPose=@n70, Alpha=1, ExternalForce="0,0,0") {
    angularDampingOverride=0.9999
    angularSpringConstant=100
    bAngularSpring=true
    bOverrideAngularDamping=true
    simulationSpace=RootRelative
    alphaBoolBlend {blendOption=Linear}
    boundBone {boneName=dyn_blade_r}
    componentAppliedLinearAccClamp=<3500 3500 3500>
    componentLinearAccScale=<1.2 1.2 1.2>
    componentLinearVelScale=<1.2 1.2 1.2>
    physicsBodyDefinitions [{
      sphereCollisionRadius=0
      boundBone {boneName=dyn_blade_r}
      boxExtents=<20 10 10>
      constraintSetup {angularLimitsMax=<1 2 30>, angularLimitsMin=<-1 -1 -30>, angularTarget=<0 0 0>}
      localJointOffset=<2 0 0>
     }]
    relativeSpaceBone {boneName=pelvis}
   }
   n72 = "Constraint - dyn_blade_r (PPP)"(ComponentPose=@n71, Alpha=1, ConstraintWeights_0=0.255, ConstraintWeights_1=0.595, ConstraintWeights_2=0.15) {
    alphaBoolBlend {blendOption=Linear}
    boneToModify {boneName=dyn_blade_r}
    constraintSetup [{targetBone {boneName=pelvis}}, {targetBone {boneName=thigh_twist_01_r}}, {targetBone {boneName=thigh_r}}]
    constraintWeights [1,1,1]
   }
   n73 = "Constraint - dyn_blade_l (PPP)"(ComponentPose=@n72, Alpha=1, ConstraintWeights_0=0.255, ConstraintWeights_1=0.595, ConstraintWeights_2=0.15) {
    alphaBoolBlend {blendOption=Linear}
    boneToModify {boneName=dyn_blade_l}
    constraintSetup [{targetBone {boneName=pelvis}}, {targetBone {boneName=thigh_twist_01_l}}, {targetBone {boneName=thigh_l}}]
    constraintWeights [1,1,1]
   }
   n74 = Get bIsSkyDiving()
   n75 = "Constraint - dyn_shouldercape_l_02 (P)"!(ComponentPose=@n73, bAlphaBoolEnabled:bool=@n74, ConstraintWeights_0=1) -> (Pose) {
    alphaInputType=Bool
    alphaBoolBlend {blendOption=Linear}
    boneToModify {boneName=dyn_shouldercape_l_02}
    constraintSetup [{targetBone {boneName=upperarm_l}}]
    constraintWeights [1]
   }
   n76 = "Constraint - dyn_shouldercape_l_bk_02 (P)"!(ComponentPose=@n75, bAlphaBoolEnabled:bool=@n74, ConstraintWeights_0=1) -> (Pose) {
    alphaInputType=Bool
    alphaBoolBlend {blendOption=Linear}
    boneToModify {boneName=dyn_shouldercape_l_bk_02}
    constraintSetup [{targetBone {boneName=upperarm_l}}]
    constraintWeights [1]
   }
   n77 = Blend Poses by bool(BlendPose_0=Component To Local(ComponentPose=@n47), BlendPose_1=Component To Local(ComponentPose=@n76), bActiveValue=Get bRigidBodyEnabled()) {blendType=Linear}
   n78 = RigidBody!(
    ComponentPose=Local To Component(LocalPose=Use cached pose 'Pose'())
    Alpha
    OverrideWorldGravity:Vector=Get GravityOverride()
    ExternalForce
   ) -> (Pose) {
    bClampLinearTranslationLimitToRefPose=true
    bOverrideWorldGravity=true
    overridePhysicsAsset=&F_MED_Knight_Cat_Cape_Physics
    simulationSpace=BaseBoneSpace
    alphaBoolBlend {blendOption=Linear}
    baseBoneRef {boneName=spine_05}
    componentAppliedLinearAccClamp=<3000 5000 5000>
    componentLinearAccScale=<0.2 0.5 0.5>
    componentLinearVelScale=<0.2 0.5 0.5>
    overrideWorldGravity {z=-980}
   }
   n79 = "Constraint - dyn_shouldercape_r_bk_01 (P)"(ComponentPose=@n78, Alpha=0.8, ConstraintWeights_0=1) {
    alphaBoolBlend {blendOption=Linear}
    boneToModify {boneName=dyn_shouldercape_r_bk_01}
    constraintSetup [{targetBone {boneName=clavicle_r}}]
    constraintWeights [1]
   }
   n80 = Component To Local(ComponentPose=@n79)
   n81 = Get bIsSkyDiving()
   n82 = Look At - Bone: dyn_shouldercape_l_bk_MSTR(
    ComponentPose=Local To Component(LocalPose=Blend Poses by bool(BlendPose_0=@n80, BlendPose_1=@n77, bActiveValue=@n81) {blendType=Linear})
    Alpha=1
   ) {
    alpha=0
    alphaBoolBlend {blendOption=Linear}
    alphaScaleBiasClamp {bInterpResult=true, interpSpeedIncreasing=20}
    boneToModify {boneName=dyn_shouldercape_l_bk_MSTR}
    lookAtLocation=<45 -16.3152 1.9334>
    lookAtTarget {boneReference {boneName=upperarm_l}}
    lookAt_Axis {axis=<1 0>}
    lookUp_Axis {axis {x=0}}
   }
   n83 = Look At - Bone: dyn_shouldercape_l_MSTR(ComponentPose=@n82, Alpha=1) {
    alpha=0
    alphaBoolBlend {blendOption=Linear}
    alphaScaleBiasClamp {bInterpResult=true, interpSpeedIncreasing=7}
    boneToModify {boneName=dyn_shouldercape_l_MSTR}
    lookAtLocation=<45 0 14.2679>
    lookAtTarget {boneReference {boneName=upperarm_l}}
    lookAt_Axis {axis=<1 0>}
    lookUp_Axis {axis {x=0}}
   }
   n84 = "Constraint - dyn_shouldercape_r_fr_02 (P)"(ComponentPose=@n83, Alpha=1, ConstraintWeights_0=1) {
    alpha=0
    alphaBoolBlend {blendOption=Linear}
    alphaScaleBiasClamp {bInterpResult=true, clampMax=0.7, interpSpeedIncreasing=20}
    boneToModify {boneName=dyn_shouldercape_r_fr_02}
    constraintSetup [{targetBone {boneName=spine_05}}]
    constraintWeights [1]
   }
   n85 = "Constraint - dyn_shouldercape_l_02 (P)"!(ComponentPose=@n84, Alpha:float=0.7, ConstraintWeights_0=1) -> (Pose) {
    alpha=0
    alphaBoolBlend {blendOption=Linear}
    alphaScaleBiasClamp {bInterpResult=true, clampMax=0.7, interpSpeedIncreasing=20}
    boneToModify {boneName=dyn_shouldercape_l_02}
    constraintSetup [{targetBone {boneName=upperarm_l}}]
    constraintWeights [1]
   }
   n86 = "Transform (Modify) Bone - Bone: dyn_shouldercape_r_fr_02"!(ComponentPose=@n85, Alpha:float=1, Rotation="15,-0.000,0.000") -> (Pose) {
    rotationMode=BMM_Additive
    rotationSpace=BCS_BoneSpace
    scaleMode=BMM_Additive
    scaleSpace=BCS_BoneSpace
    translationMode=BMM_Additive
    translationSpace=BCS_BoneSpace
    alphaBoolBlend {blendOption=Linear}
    alphaScaleBiasClamp {interpSpeedIncreasing=20}
    boneToModify {boneName=dyn_shouldercape_r_fr_02}
    rotation {pitch=15}
   }
   Save cached pose 'SavedPose27606'(Pose=Component To Local(ComponentPose=@n86)) {cacheName=SavedPose27606}
  ]}
 EventGraph {nodes [
   Event BlueprintUpdateAnimation()
   > n1 = Sequence!(execute) -> (then_0, then_1, then_2:Exec, then_3:Exec, then_4:Exec, then_5:Exec, then_6:Exec, then_7:Exec, then_8:Exec, then_9:Exec)
   > n2 = "Gravity Override State Function (Parameter Asset) (Nativized)"(
    CharacterPartAnimInstance=Self-Reference()
    GameTimeInSeconds=GetGameTimeInSeconds()
    Parameters=&/F_MED_Knight_Cat_Parameters
   )
   > Set GravityOverride(GravityOverride=@n2)
   n3 = BreakTransform(InTransform=GetDeltaTransformFromRefPose(self=GetOwnerSkeletalMeshComponent(), BoneName=thigh_r, BaseName=pelvis))
   Set thigh_rot_up_r(execute=@n1.then_1, thigh_rot_up_r:double=MapRangeClamped(Value=BreakRotator(InRot=@n3.Rotation).Yaw, InRangeA=-15, InRangeB=-100)) -> (Output_Get:double)
   n4 = BreakTransform(InTransform=GetDeltaTransformFromRefPose(self=GetOwnerSkeletalMeshComponent(), BoneName=thigh_l, BaseName=pelvis))
   > Set thigh_rot_up_l(thigh_rot_up_l:double=MapRangeClamped(Value=BreakRotator(InRot=@n4.Rotation).Yaw, InRangeA=-10, InRangeB=-100)) -> (Output_Get:double)
   n5 = Branch(execute=@n1.then_2, Condition="float > float"(A=Get PawnSpeed2D(), B=0) {is=title K2Node_CallFunction})
   > Set Forward_Direct(Forward_Direct:double=100) -> (Output_Get:double)
   Set Forward_Direct(execute=@n5.else, Forward_Direct=0)
   n6 = Branch(execute=@n1.then_3, Condition=Get bIsInFrontEnd())
   > Set FrontEnd Alpha(FrontEnd Alpha:double=1) -> (Output_Get:double)
   Set FrontEnd Alpha(execute=@n6.else, FrontEnd Alpha=0)
   n7 = Branch(execute=@n1.then_4, Condition=Get bIsPlayingMeleeAnim())
   > Set Is_Swinging(Is_Swinging:double=1) -> (Output_Get:double)
   Set Is_Swinging(execute=@n7.else, Is_Swinging=0)
   n8 = CheckForBackbling(self=Self-Reference())
   n9 = OR Boolean(A=@n8, B=@n8.bIsBackblingACape)
   Set wearing backpack Bool(execute=@n1.then_6, wearing backpack Bool:bool=@n9) -> (Output_Get:bool)
   > n10 = Set wearingBackpack(wearingBackpack:double=SelectFloat(A=0, B="To Float (Boolean)"(InBool=@n9), bPickA=false)) -> (Output_Get:double)
   > Set noBackpack(noBackpack:double="To Float (Boolean)"(InBool="Not Equal (Float)"(A=@n10, B=1))) -> (Output_Get:double)
   n11 = Get PawnMovementDirectionAzimuth() -> (PawnMovementDirectionAzimuth:float)
   n12 = float - float(A=MapRangeClamped(Value=@n11, InRangeA=-45, InRangeB=-90), B=MapRangeClamped(Value=@n11, InRangeA=-90, InRangeB=-135)) {is=title K2Node_CallFunction}
   n13 = Get bIsMoving2D()
   Set Left_Direct(execute=@n1.then_7, Left_Direct:double=SelectFloat(A="Absolute (Float)"(A=@n12), B=0, bPickA=@n13)) -> (Output_Get:double)
   n14 = Get PawnMovementDirectionAzimuth()
   n15 = float - float(A=MapRangeClamped(Value=@n14, InRangeA=45, InRangeB=90), B=MapRangeClamped(Value=@n14, InRangeA=90, InRangeB=135))
   n16 = Get bIsCrouching()
   n17 = Get bIsMoving2D()
   Set Right_Direct(execute=@n1.then_8, Right_Direct:double=SelectFloat(A="Absolute (Float)"(A=@n15), B=0, bPickA=AND Boolean(A=@n16, B=@n17))) -> (Output_Get:double)
   n18 = Get bRigidBodyEnabled()
   Set frontend_noRBAN(execute=@n1.then_9, frontend_noRBAN:bool=AND Boolean(A=Get bIsInFrontEnd(), B=NOT Boolean(A=@n18))) -> (Output_Get:bool)
   n19 = GetOwnerSkeletalMeshComponent()
   n20 = BreakRotator(InRot=BreakTransform(InTransform=GetDeltaTransformFromRefPose(self=@n19, BoneName=upperarm_r, BaseName=spine_05)).Rotation)
   Set arm_up_r(
    execute=@n1.then_5
    arm_up_r:double=SelectFloat(A=MapRangeClamped(Value=@n20.Yaw, InRangeA=34, InRangeB=100), B=0, bPickA=Get bIsTargeting())
   ) -> (Output_Get:double)
   n21 = BreakTransform(InTransform=GetDeltaTransformFromRefPose(self=GetOwnerSkeletalMeshComponent(), BoneName=upperarm_l, BaseName=spine_05))
   > Set arm_up_l(arm_up_l:double=MapRangeClamped(Value=BreakRotator(InRot=@n21.Rotation).Yaw, InRangeA=25, InRangeB=60)) -> (Output_Get:double)
   n22 = GetOwnerSkeletalMeshComponent()
   n23 = BreakRotator(InRot=BreakTransform(InTransform=GetDeltaTransformFromRefPose(self=@n22, BoneName=upperarm_r, BaseName=spine_01)).Rotation)
   > Set arm_down_r(arm_down_r:double=SelectFloat(A=MapRangeClamped(Value=@n23.Yaw, InRangeA=120, InRangeB=8), B=0, bPickA=Get bIsTargeting())) -> (Output_Get:double)
   n24 = BreakTransform(InTransform=GetDeltaTransformFromRefPose(self=GetOwnerSkeletalMeshComponent(), BoneName=head, BaseName=spine_05))
   > Set Head Up(Head Up:double=MapRangeClamped(Value=BreakRotator(InRot=@n24.Rotation).Yaw, InRangeA=-5, InRangeB=-40)) -> (Output_Get:double)
   n25 = BreakTransform(InTransform=GetDeltaTransformFromRefPose(self=GetOwnerSkeletalMeshComponent(), BoneName=spine_05, BaseName=pelvis))
   > Set Twist(Twist:double=MapRangeClamped(Value=BreakRotator(InRot=@n25.Rotation), InRangeA=-2, InRangeB=-5)) -> (Output_Get:double)
   n26 = Event OnParachuteOpened()
   n27 = Get Dynamic Material Instance(execute=[@n26.then, @n40.then], SkeletalMeshComponent=GetPartSkeletalMeshComponent())
   > n28 = Sequence!(execute) -> (then_0, then_1)
   n29 = Make Array("[0]"=(TagName="Cosmetics.MetaTags.KnightCat"))
   n30 = PickaxeItemDefinitionHasAnyMetaTag(Tags=@n29, bRequiresEquipped=true)
   n31 = Branch(execute=[@n39.then, @n28.then_0], Condition=@n30)
   > n32 = Branch(Condition=Get bIsSkyDiving())
   SetScalarParameterValue(execute=[@n37.then, @n35.else, @n32.then, @n31.else], self=@n27, ParameterName=Hide Element 01, Value=0)
   SetScalarParameterValue(execute=[@n34.then, @n35.then, @n32.else], self=@n27, ParameterName=Hide Element 01, Value=1)
   n33 = Branch(
    execute=@n28.then_1
    Condition=GliderItemDefinitionHasAnyMetaTag!(Tags:Array of Gameplay Tag Structures=@n29, bRequiresOpened:bool=true) -> (ReturnValue:bool)
   )
   > n34 = Delay(Duration=0.2)
   n35 = Branch(execute=[@n38.then, @n33.else], Condition=@n30)
   Event OnCharacterCustomizationLoaded()
   > n36 = Branch(Condition=CharacterItemDefinitionHasAnyMetaTag(Tags=Make Array("[0]"=(TagName="Cosmetics.Character.RestrictCape"))))
   > Set cape_active(cape_active:bool=true) -> (Output_Get:bool)
   Set cape_active(execute=@n36.else, cape_active=false)
   n37 = Event OnBeginSkydiving()
   n38 = Event OnEndSkydiving()
   n39 = Event OnParachuteClosed()
   n40 = Event OnWeaponChanged()
   TryGetPawnOwner()
   AND Boolean(A=@n13, B=NOT Boolean(A=Get bIsCrouching()))
   Get bIsPlayingEmote()
  ]}
]
variables [
 arm_down_r {type=number}
 arm_up_l {type=number}
 arm_up_r {type=number}
 cape_active {type=boolean}
 Forward_Direct {type=number}
 FrontEnd Alpha {type=number}
 frontend_noRBAN {type=boolean}
 GravityOverride {type=Vector}
 Head Up {type=number}
 Is_Swinging {type=number}
 Left_Direct {type=number}
 noBackpack {type=number}
 Right_Direct {type=number}
 thigh_rot_up_l {type=number}
 thigh_rot_up_r {type=number}
 Twist {type=number}
 wearing backpack Bool {type=boolean}
 wearingBackpack {type=number}
]
