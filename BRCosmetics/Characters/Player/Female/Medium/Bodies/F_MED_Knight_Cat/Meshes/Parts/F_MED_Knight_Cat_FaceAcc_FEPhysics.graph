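# Physics asset for the F_MED_Knight_Cat face accessory mesh, providing the bodies and joints to simulate a hair braid.
# Nine braid capsules chain from hair_braid_MSTR with tight twist and moderate swing limits, narrowing in swing at the last link.
# Capsules on the head, neck, upper spine and left upper arm give the body's shape around the braid.
# Those body joints and the master-to-head joint leave swing and twist free.
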PhysicsAsset /BRCosmetics/Characters/Player/Female/Medium/Bodies/F_MED_Knight_Cat/Meshes/Parts/F_MED_Knight_Cat_FaceAcc_FEPhysics
bodies {
 hair_braid_01 [{shapeType=Capsule, length=6.0961, radius=1.3598, shapeName=hair_braid_01_capsule, center=<2 0.1621 -0.0001>, rotation=<rot 90 0.844 0.8441>}]
 hair_braid_02 [{shapeType=Capsule, length=2.8, radius=1.3598, shapeName=hair_braid_02_capsule, center=<1.8497 -0.2102 0>, rotation=<rot 90 14.0363 19.0362>}]
 hair_braid_03 [{shapeType=Capsule, length=2.5896, radius=1.3598, shapeName=hair_braid_03_capsule, center=<1.4653 0.3853 0>, rotation=<rot 90 -90 -95.0002>}]
 hair_braid_04 [{shapeType=Capsule, length=1.684, radius=1.3598, shapeName=hair_braid_04_capsule, center=<1.0435 0.1425 -0.0001>, rotation=<rot 90 0 0>}]
 hair_braid_05 [{shapeType=Capsule, length=3.8996, radius=1.7714, shapeName=hair_braid_05_capsule, center=<3.4115 0.6074 0>, rotation=<rot 90 90 84.9998>}]
 hair_braid_06 [{shapeType=Capsule, length=1.1866, radius=1.3927, shapeName=hair_braid_06_capsule, center=<1.3382 -0.1488 -0.0001>, rotation=<rot 90 0 0>}]
 hair_braid_07 [{shapeType=Capsule, length=0.7986, radius=2.7701, shapeName=hair_braid_07_capsule, center=<2.6344 0.6032 0>, rotation=<rot 90 90 79.9999>}]
 hair_braid_08 [{shapeType=Capsule, length=3.0109, radius=1.3829, shapeName=hair_braid_08_capsule, center=<2.6045 0.3885 0>, rotation=<rot 90 -90 -95>}]
 hair_braid_09 [{shapeType=Capsule, length=5.6609, radius=1.8173, shapeName=hair_braid_09_capsule, center=<4.5077 -0.8057 0>, rotation=<rot 90 -90 -79.9999>}]
 hair_braid_MSTR [{shapeType=Capsule, length=0.505, radius=0.505, shapeName=hair_braid_MSTR_capsule, rotation=<rot 0 -90 0>}]
 head [{shapeType=Capsule, length=9.4063, radius=5.172, shapeName=head_capsule, rotation=<rot 90 0 0>}]
 neck_01 [{shapeType=Capsule, length=8.8102, radius=5, shapeName=neck_01_capsule}]
 spine_03 [{shapeType=Capsule, length=22.3443, radius=9.62, shapeName=spine_03_capsule, center=<-10.9267 0 0>}]
 spine_04 [{shapeType=Capsule, length=29.5285, radius=7.431, shapeName=spine_04_capsule, center=<-4.0028 1.6313 0>, rotation=<rot 0 -89.9999 0>}]
 spine_05 [{shapeType=Capsule, length=25.2346, radius=7.3396, shapeName=spine_05_capsule, center=<0 0.7856 0>}]
 upperarm_l [{shapeType=Capsule, length=29.9419, radius=5.8333, shapeName=upperarm_l_capsule, center=<2.3602 0 2.9699>, rotation=<rot -39.9992 0 0>}]
}
constraints [
 {
  bone1Name=head
  bone2Name=neck_01
  swing1LimitDegrees=45
  swing1Motion=Free
  swing2LimitDegrees=45
  swing2Motion=Free
  twistLimitDegrees=45
  twistMotion=Free
 }
 {
  bone1Name=neck_01
  bone2Name=spine_05
  swing1LimitDegrees=45
  swing1Motion=Free
  swing2LimitDegrees=45
  swing2Motion=Free
  twistLimitDegrees=45
  twistMotion=Free
 }
 {
  bone1Name=upperarm_l
  bone2Name=spine_05
  swing1LimitDegrees=45
  swing1Motion=Free
  swing2LimitDegrees=45
  swing2Motion=Free
  twistLimitDegrees=45
  twistMotion=Free
 }
 {bone1Name=hair_braid_01, bone2Name=hair_braid_MSTR, swing1LimitDegrees=25, swing2LimitDegrees=25, twistLimitDegrees=5}
 {bone1Name=hair_braid_02, bone2Name=hair_braid_01, swing1LimitDegrees=25, swing2LimitDegrees=25, twistLimitDegrees=5}
 {bone1Name=hair_braid_03, bone2Name=hair_braid_02, swing1LimitDegrees=25, swing2LimitDegrees=25, twistLimitDegrees=5}
 {bone1Name=hair_braid_04, bone2Name=hair_braid_03, swing1LimitDegrees=25, swing2LimitDegrees=25, twistLimitDegrees=5}
 {bone1Name=hair_braid_05, bone2Name=hair_braid_04, swing1LimitDegrees=25, swing2LimitDegrees=25, twistLimitDegrees=5}
 {bone1Name=hair_braid_06, bone2Name=hair_braid_05, swing1LimitDegrees=25, swing2LimitDegrees=25, twistLimitDegrees=5}
 {bone1Name=hair_braid_07, bone2Name=hair_braid_06, swing1LimitDegrees=25, swing2LimitDegrees=25, twistLimitDegrees=5}
 {bone1Name=hair_braid_08, bone2Name=hair_braid_07, swing1LimitDegrees=25, swing2LimitDegrees=25, twistLimitDegrees=5}
 {
  bone1Name=spine_05
  bone2Name=spine_04
  swing1LimitDegrees=45
  swing1Motion=Free
  swing2LimitDegrees=45
  swing2Motion=Free
  twistLimitDegrees=45
  twistMotion=Free
 }
 {
  bone1Name=spine_04
  bone2Name=spine_03
  swing1LimitDegrees=45
  swing1Motion=Free
  swing2LimitDegrees=45
  swing2Motion=Free
  twistLimitDegrees=45
  twistMotion=Free
 }
 {
  bone1Name=hair_braid_MSTR
  bone2Name=head
  swing1LimitDegrees=45
  swing1Motion=Free
  swing2LimitDegrees=45
  swing2Motion=Free
  twistLimitDegrees=45
  twistMotion=Free
 }
 {bone1Name=hair_braid_09, bone2Name=hair_braid_08, swing1LimitDegrees=15, swing2LimitDegrees=15, twistLimitDegrees=5}
]
