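# Physics asset for the Knight Cat female character's FaceAcc part, giving physics bodies to a hair braid and right-side bang.
# A nine-segment braid and a shorter bang chain use tight swing and twist limits, hanging from root bodies linked to the head.
# Head, upper spine and left upper arm capsules, plus a box on a left cape bone, form the surrounding body shapes, joined by free angular constraints.
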PhysicsAsset /BRCosmetics/Characters/Player/Female/Medium/Bodies/F_MED_Knight_Cat/Meshes/Parts/F_MED_Knight_Cat_FaceAcc_FE_Physics
bodies {
 dyn_cape_l_frmid_02 [{
   shapeType=Box
   extentX=20
   extentY=5
   extentZ=33
   shapeName=dyn_cape_l_frmid_02_box
   center=<4.2866 4.6706 -8.9761>
   rotation=<rot 57.1306 -18.3672 -33.0781>
  }]
 hair_bang_r_01 [{shapeType=Capsule, length=0.505, radius=0.505, shapeName=hair_bang_r_01_capsule, rotation=<rot 0 -90 0>}]
 hair_bang_r_02 [{shapeType=Capsule, length=2, radius=1.5, shapeName=hair_bang_r_02_capsule, center=<1.5 0 0>, rotation=<rot 90 0 0>}]
 hair_bang_r_04 [{shapeType=Capsule, length=5, radius=1.5, shapeName=hair_bang_r_04_capsule, center=<2 0 0>, rotation=<rot 90 0 0>}]
 hair_bang_r_06 [{shapeType=Capsule, length=5, radius=1.5, shapeName=hair_bang_r_06_capsule, center=<2 0 0>, rotation=<rot 90 -7.125 -17.1252>}]
 hair_braid_01 [{shapeType=Capsule, length=6.0961, radius=1.3598, shapeName=hair_braid_01_capsule, center=<2 0.1621 -0.0001>, rotation=<rot 90 0.844 0.8441>}]
 hair_braid_02 [{shapeType=Capsule, length=3, radius=2.5, shapeName=hair_braid_02_capsule, center=<2 1 0>, rotation=<rot 90 0 0>}]
 hair_braid_03 [{shapeType=Capsule, length=3, radius=2.5, shapeName=hair_braid_03_capsule, center=<2 1 0>, rotation=<rot 90 0 0>}]
 hair_braid_04 [{shapeType=Sphere, radius=1.4776, shapeName=hair_braid_04_sphere, center=<0.9419 0.2925 -0.0078>}]
 hair_braid_05 [{shapeType=Capsule, length=2.8299, radius=2.75, shapeName=hair_braid_05_capsule, center=<2 1 0>, rotation=<rot 90 0 0>}]
 hair_braid_06 [{shapeType=Sphere, radius=1.1167, shapeName=hair_braid_06_sphere, center=<1.2414 0 0>}]
 hair_braid_07 [{shapeType=Capsule, length=0.7237, radius=2.5, shapeName=hair_braid_07_capsule, center=<2 1 0>, rotation=<rot 90 0 0>}]
 hair_braid_08 [{shapeType=Capsule, length=3.0109, radius=1, shapeName=hair_braid_08_capsule, center=<2 1 0>, rotation=<rot 90 0 0>}]
 hair_braid_09 [{shapeType=Capsule, length=4.5767, radius=2.25, shapeName=hair_braid_09_capsule, center=<2 1 0>, rotation=<rot 90 0 0>}]
 hair_braid_MSTR [{shapeType=Capsule, length=0.505, radius=0.505, shapeName=hair_braid_MSTR_capsule, rotation=<rot 0 -90 0>}]
 head [{shapeType=Capsule, length=9.4063, radius=5.172, shapeName=head_capsule, rotation=<rot 90 0 0>}]
 spine_03 [{shapeType=Capsule, length=31.8928, radius=12.6515, shapeName=spine_03_capsule, center=<-3.0817 -1.1497 0>}]
 spine_04 [{shapeType=Capsule, length=25.6943, radius=11.2652, shapeName=spine_04_capsule, center=<-2.9002 0.4095 0>}]
 spine_05 [{shapeType=Capsule, length=23.3458, radius=9.2284, shapeName=spine_05_capsule, center=<1.1727 -2.5755 0>}]
 upperarm_l [{
   shapeType=Capsule
   length=14.4074
   radius=5.1261
   shapeName=upperarm_l_capsule
   center=<0.2644 0.1982 -1.3565>
   rotation=<rot -29.9992 0.0006 -0.0003>
  }]
}
constraints [
 {
  bone1Name=head
  bone2Name=spine_05
  swing1LimitDegrees=45
  swing1Motion=Free
  swing2LimitDegrees=45
  swing2Motion=Free
  twistLimitDegrees=45
  twistMotion=Free
 }
 {
  bone1Name=upperarm_l
  bone2Name=spine_05
  swing1LimitDegrees=45
  swing1Motion=Free
  swing2LimitDegrees=45
  swing2Motion=Free
  twistLimitDegrees=45
  twistMotion=Free
 }
 {bone1Name=hair_braid_01, bone2Name=hair_braid_MSTR, swing1LimitDegrees=20, swing2LimitDegrees=20, twistLimitDegrees=5}
 {bone1Name=hair_braid_02, bone2Name=hair_braid_01, swing1LimitDegrees=20, swing2LimitDegrees=20, twistLimitDegrees=5}
 {bone1Name=hair_braid_03, bone2Name=hair_braid_02, swing1LimitDegrees=20, swing2LimitDegrees=20, twistLimitDegrees=5}
 {bone1Name=hair_braid_04, bone2Name=hair_braid_03, swing1LimitDegrees=20, swing2LimitDegrees=20, twistLimitDegrees=5}
 {bone1Name=hair_braid_05, bone2Name=hair_braid_04, swing1LimitDegrees=20, swing2LimitDegrees=20, twistLimitDegrees=5}
 {bone1Name=hair_braid_06, bone2Name=hair_braid_05, swing1LimitDegrees=20, swing2LimitDegrees=20, twistLimitDegrees=5}
 {bone1Name=hair_braid_07, bone2Name=hair_braid_06, swing1LimitDegrees=20, swing2LimitDegrees=20, twistLimitDegrees=5}
 {bone1Name=hair_braid_08, bone2Name=hair_braid_07, swing1LimitDegrees=20, swing2LimitDegrees=20, twistLimitDegrees=5}
 {
  bone1Name=spine_05
  bone2Name=spine_04
  swing1LimitDegrees=45
  swing1Motion=Free
  swing2LimitDegrees=45
  swing2Motion=Free
  twistLimitDegrees=45
  twistMotion=Free
 }
 {
  bone1Name=spine_04
  bone2Name=spine_03
  swing1LimitDegrees=45
  swing1Motion=Free
  swing2LimitDegrees=45
  swing2Motion=Free
  twistLimitDegrees=45
  twistMotion=Free
 }
 {
  bone1Name=hair_braid_MSTR
  bone2Name=head
  swing1LimitDegrees=45
  swing1Motion=Free
  swing2LimitDegrees=45
  swing2Motion=Free
  twistLimitDegrees=45
  twistMotion=Free
 }
 {bone1Name=hair_braid_09, bone2Name=hair_braid_08, swing1LimitDegrees=20, swing2LimitDegrees=20, twistLimitDegrees=5}
 {
  bone1Name=dyn_cape_l_frmid_02
  bone2Name=upperarm_l
  swing1LimitDegrees=45
  swing1Motion=Free
  swing2LimitDegrees=45
  swing2Motion=Free
  twistLimitDegrees=45
  twistMotion=Free
 }
 {
  bone1Name=hair_bang_r_01
  bone2Name=head
  swing1LimitDegrees=45
  swing1Motion=Free
  swing2LimitDegrees=45
  swing2Motion=Free
  twistLimitDegrees=45
  twistMotion=Free
 }
 {bone1Name=hair_bang_r_02, bone2Name=hair_bang_r_01, swing1LimitDegrees=30, swing2LimitDegrees=20, twistLimitDegrees=1}
 {bone1Name=hair_bang_r_04, bone2Name=hair_bang_r_02, swing1LimitDegrees=10, swing2LimitDegrees=10, twistLimitDegrees=1}
 {bone1Name=hair_bang_r_06, bone2Name=hair_bang_r_04, swing1LimitDegrees=10, swing2LimitDegrees=10, twistLimitDegrees=1}
]
