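# Anim Blueprint for the hair of the Knight Cat female character's face accessory, copying the body pose and adding secondary motion.
# AnimDynamics sways tufts and strands into a cached base pose, which feeds RigidBody simulations chosen by bIsInFrontEnd, with bone offset or braid constraints beforehand.
# The event graph computes gravity override and head/arm corrective alphas, but their consumers, like other bang, braid and skydiving chains, never reach the output.
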AnimBlueprint /BRCosmetics/Characters/Player/Female/Medium/Bodies/F_MED_Knight_Cat/Meshes/Parts/F_MED_Knight_Cat_FaceAcc_AnimBP
target_skeleton=&/F_MED_Knight_Cat_FaceAcc_Skeleton
graphs [
 AnimGraph {nodes [
   Anim Dynamics - Bone: hair_bang_r_01(Alpha=1, ExternalForce="0,0,0") {
    angularDampingOverride=5
    angularSpringConstant=1
    bAngularSpring=true
    bLinearSpring=true
    bUseSphericalLimits=true
    linearSpringConstant=1
    simulationSpace=RootRelative
    alphaBoolBlend {blendOption=Linear}
    boundBone {boneName=hair_bang_r_01}
    componentAppliedLinearAccClamp=<3500 3500 3500>
    componentLinearAccScale=<1.2 1.2 1.2>
    componentLinearVelScale=<1.2 1.2 1.2>
    physicsBodyDefinitions [{
      boundBone {boneName=hair_bang_r_01}
      boxExtents=<10 10 10>
      constraintSetup {angularLimitsMax=<0 0 0>, angularLimitsMin=<0 0 0>, angularTarget=<0 0 0>}
      localJointOffset=<0 0 0>
     }]
    relativeSpaceBone {boneName=dyn_hair_simspace}
    sphericalLimits [{limitRadius=8, drivingBone {boneName=head}, sphereLocalOffset=<0 0 0>}]
   }
   n1 = Anim Dynamics - Bone: hair_bang_r_02(Alpha=1, ExternalForce="0,0,0") {
    angularDampingOverride=5
    angularSpringConstant=1
    bAngularSpring=true
    bLinearSpring=true
    bUseSphericalLimits=true
    linearSpringConstant=1
    simulationSpace=RootRelative
    alphaBoolBlend {blendOption=Linear}
    boundBone {boneName=hair_bang_r_02}
    componentAppliedLinearAccClamp=<3500 3500 3500>
    componentLinearAccScale=<1.2 1.2 1.2>
    componentLinearVelScale=<1.2 1.2 1.2>
    physicsBodyDefinitions [{
      sphereCollisionRadius=0
      boundBone {boneName=hair_bang_r_02}
      boxExtents=<50 50 50>
      constraintSetup {angularLimitsMax=<1 10 15>, angularLimitsMin=<-1 -30 -15>, angularTarget=<0 0 0>}
      localJointOffset=<5 0 0>
     }]
    relativeSpaceBone {boneName=dyn_hair_simspace}
    sphericalLimits [{limitRadius=5, drivingBone {boneName=head}, sphereLocalOffset=<0 0 2.5>}]
   }
   n2 = Anim Dynamics - Bone: hair_bang_r_03(ComponentPose=@n1, Alpha=1, ExternalForce="0,0,0") {
    angularDampingOverride=5
    angularSpringConstant=1
    bAngularSpring=true
    bLinearSpring=true
    bUseSphericalLimits=true
    linearSpringConstant=1
    simulationSpace=RootRelative
    alphaBoolBlend {blendOption=Linear}
    boundBone {boneName=hair_bang_r_03}
    componentAppliedLinearAccClamp=<3500 3500 3500>
    componentLinearAccScale=<1.2 1.2 1.2>
    componentLinearVelScale=<1.2 1.2 1.2>
    physicsBodyDefinitions [{
      sphereCollisionRadius=0
      boundBone {boneName=hair_bang_r_03}
      boxExtents=<50 50 50>
      constraintSetup {angularLimitsMax=<1 5 15>, angularLimitsMin=<-1 -30 -15>, angularTarget=<0 0 0>}
      localJointOffset=<5 0 0>
     }]
    relativeSpaceBone {boneName=dyn_hair_simspace}
    sphericalLimits [{limitRadius=7, drivingBone {boneName=head}, sphereLocalOffset=<-5 0 0>}]
   }
   n3 = Anim Dynamics - Bone: hair_bang_r_04(ComponentPose=@n2, Alpha=1, ExternalForce="0,0,0") {
    angularDampingOverride=5
    angularSpringConstant=1
    bAngularSpring=true
    bLinearSpring=true
    bUseSphericalLimits=true
    linearSpringConstant=1
    simulationSpace=RootRelative
    alphaBoolBlend {blendOption=Linear}
    boundBone {boneName=hair_bang_r_04}
    componentAppliedLinearAccClamp=<3500 3500 3500>
    componentLinearAccScale=<1.2 1.2 1.2>
    componentLinearVelScale=<1.2 1.2 1.2>
    physicsBodyDefinitions [{
      sphereCollisionRadius=0
      boundBone {boneName=hair_bang_r_04}
      boxExtents=<50 50 50>
      constraintSetup {angularLimitsMax=<1 5 15>, angularLimitsMin=<-1 -30 -15>, angularTarget=<0 0 0>}
      localJointOffset=<5 0 0>
     }]
    relativeSpaceBone {boneName=dyn_hair_simspace}
    sphericalLimits [{limitRadius=8, drivingBone {boneName=head}, sphereLocalOffset=<-5 0 0>}]
   }
   Anim Dynamics - Bone: hair_bang_r_05(ComponentPose=@n3, Alpha=1, ExternalForce="0,0,0") {
    angularDampingOverride=5
    angularSpringConstant=1
    bAngularSpring=true
    bLinearSpring=true
    bUseSphericalLimits=true
    linearSpringConstant=1
    simulationSpace=RootRelative
    alphaBoolBlend {blendOption=Linear}
    boundBone {boneName=hair_bang_r_05}
    componentAppliedLinearAccClamp=<3500 3500 3500>
    componentLinearAccScale=<1.2 1.2 1.2>
    componentLinearVelScale=<1.2 1.2 1.2>
    physicsBodyDefinitions [{
      sphereCollisionRadius=0
      boundBone {boneName=hair_bang_r_05}
      boxExtents=<50 50 50>
      constraintSetup {angularLimitsMax=<1 5 15>, angularLimitsMin=<-1 -30 -15>, angularTarget=<0 0 0>}
      localJointOffset=<5 0 0>
     }]
    relativeSpaceBone {boneName=dyn_hair_simspace}
    sphericalLimits [{limitRadius=5, drivingBone {boneName=clavicle_r}, sphereLocalOffset=<0 -3.5 -4>}]
   }
   n4 = RigidBody!(
    ComponentPose=Local To Component(LocalPose=Use cached pose 'Base'())
    Alpha
    OverrideWorldGravity:Vector=Get GravityOverride()
    ExternalForce
   ) -> (Pose) {
    bClampLinearTranslationLimitToRefPose=true
    bOverrideWorldGravity=true
    overridePhysicsAsset=&/F_MED_Knight_Cat_FaceAcc_Physics
    simulationSpace=BaseBoneSpace
    alphaBoolBlend {blendOption=Linear}
    baseBoneRef {boneName=spine_05}
    componentAppliedLinearAccClamp=<3000 3000 3000>
    componentLinearAccScale=<0.25 0.5 0.5>
    componentLinearVelScale=<0.25 0.5 0.5>
   }
   n5 = "Transform (Modify) Bone - Bone: hair_braid_MSTR"!(ComponentPose=@n4, Alpha=1, Translation="0,0,0", Rotation="34.470,-28.848,14.345") -> (Pose) {
    rotationMode=BMM_Additive
    rotationSpace=BCS_BoneSpace
    scaleMode=BMM_Additive
    scaleSpace=BCS_BoneSpace
    translationMode=BMM_Additive
    translationSpace=BCS_BoneSpace
    alphaBoolBlend {blendOption=Linear}
    boneToModify {boneName=hair_braid_MSTR}
    rotation=<rot 34.47 -28.848 14.345>
   }
   Blend Poses by bool(BlendPose_0=Component To Local(ComponentPose=@n5), bActiveValue=Get bIsSkyDiving()) {blendType=Linear}
   Blend Poses by bool(bActiveValue=Get bRigidBodyEnabled()) {blendType=Linear}
   n6 = RigidBody!(ComponentPose=Local To Component(LocalPose=Use cached pose 'Base'()), Alpha, ExternalForce) -> (Pose) {
    bClampLinearTranslationLimitToRefPose=true
    bOverrideWorldGravity=true
    overridePhysicsAsset=&/F_MED_Knight_Cat_FaceAcc_FEPhysics
    simulationSpace=BaseBoneSpace
    alphaBoolBlend {blendOption=Linear}
    baseBoneRef {boneName=neck_01}
    componentAppliedLinearAccClamp=<1000 1000 1000>
    componentLinearAccScale=<0.5 0.5 0.5>
    componentLinearVelScale=<0.5 0.5 0.5>
    overrideWorldGravity {z=-980}
   }
   n7 = "Constraint - hair_braid_06 (P)"(ComponentPose=@n6, Alpha=0.5, ConstraintWeights_0=1) {
    alphaBoolBlend {blendOption=Linear}
    boneToModify {boneName=hair_braid_06}
    constraintSetup [{targetBone {boneName=spine_04}}]
    constraintWeights [1]
   }
   n8 = "Transform (Modify) Bone - Bone: hair_braid_03"!(ComponentPose=@n7, Alpha=1, Translation="0,-1,0.000", Rotation="-6,3,-0.000") -> (Pose) {
    rotationMode=BMM_Replace
    rotationSpace=BCS_BoneSpace
    scaleMode=BMM_Additive
    scaleSpace=BCS_BoneSpace
    translationMode=BMM_Additive
    translationSpace=BCS_BoneSpace
    alphaBoolBlend {blendOption=Linear}
    boneToModify {boneName=hair_braid_03}
    rotation {pitch=-6, yaw=3}
    translation {y=-1}
   }
   Blend Poses by bool(BlendPose_0=Component To Local(ComponentPose=@n8), bActiveValue=Get bIsInFrontEnd()) {blendType=Linear}
   n9 = "Transform (Modify) Bone - Bone: hair_bang_r_MSTR"!(ComponentPose=Local To Component(), Alpha=Get arm_up_r() -> (arm_up_r:double), Translation="0,0,0", Rotation="8,0,0") -> (Pose) {
    rotationMode=BMM_Replace
    rotationSpace=BCS_BoneSpace
    scaleMode=BMM_Additive
    scaleSpace=BCS_BoneSpace
    translationMode=BMM_Additive
    translationSpace=BCS_BoneSpace
    alphaBoolBlend {blendOption=Linear}
    alphaScaleBiasClamp {bInterpResult=true, interpSpeedIncreasing=20}
    boneToModify {boneName=hair_bang_r_MSTR}
    rotation {pitch=8}
   }
   n10 = Get arm_up_l() -> (arm_up_l:double)
   n11 = "Transform (Modify) Bone - Bone: hair_braid_02"(ComponentPose=@n9, Alpha=@n10, Translation="0,0,0", Rotation="-5,10,-4", Scale="1,1,1") {
    rotationMode=BMM_Replace
    rotationSpace=BCS_BoneSpace
    scaleMode=BMM_Additive
    scaleSpace=BCS_BoneSpace
    translationMode=BMM_Replace
    translationSpace=BCS_BoneSpace
    alphaBoolBlend {blendOption=Linear}
    alphaScaleBiasClamp {bInterpResult=true, interpSpeedDecreasing=5, interpSpeedIncreasing=7}
    boneToModify {boneName=hair_braid_02}
    rotation=<rot -5 10 -4>
   }
   n12 = "Transform (Modify) Bone - Bone: hair_braid_07"(ComponentPose=@n11, Alpha=@n10, Translation="0,0,0", Rotation="-10.000,0.000,0.000", Scale="1,1,1") {
    rotationMode=BMM_Replace
    rotationSpace=BCS_BoneSpace
    scaleMode=BMM_Additive
    scaleSpace=BCS_BoneSpace
    translationMode=BMM_Replace
    translationSpace=BCS_BoneSpace
    alphaBoolBlend {blendOption=Linear}
    alphaScaleBiasClamp {bInterpResult=true, interpSpeedDecreasing=5, interpSpeedIncreasing=7}
    boneToModify {boneName=hair_braid_07}
    rotation {pitch=-10}
   }
   n13 = Get Head Up() -> (Head Up:double)
   n14 = "Transform (Modify) Bone - Bone: hair_bang_r_01"(ComponentPose=@n12, Alpha=@n13, Translation="0,0,0", Rotation="0.000,15,0.000", Scale="1,1,1") {
    rotationMode=BMM_Replace
    rotationSpace=BCS_BoneSpace
    scaleMode=BMM_Additive
    scaleSpace=BCS_BoneSpace
    translationMode=BMM_Replace
    translationSpace=BCS_BoneSpace
    alphaBoolBlend {blendOption=Linear}
    alphaScaleBiasClamp {bInterpResult=true, interpSpeedDecreasing=5, interpSpeedIncreasing=7}
    boneToModify {boneName=hair_bang_r_01}
    rotation {yaw=15}
   }
   n15 = "Transform (Modify) Bone - Bone: hair_braid_MSTR"(ComponentPose=@n14, Alpha=@n13, Translation="-0.000,-0.5,0", Rotation="-15,0.000,0.000") {
    rotationMode=BMM_Replace
    rotationSpace=BCS_BoneSpace
    scaleMode=BMM_Additive
    scaleSpace=BCS_BoneSpace
    translationMode=BMM_Replace
    translationSpace=BCS_BoneSpace
    alphaBoolBlend {blendOption=Linear}
    alphaScaleBiasClamp {bInterpResult=true, interpSpeedDecreasing=5, interpSpeedIncreasing=7}
    boneToModify {boneName=hair_braid_MSTR}
    rotation {pitch=-15}
    translation {y=-0.5}
   }
   n16 = "Transform (Modify) Bone - Bone: hair_braid_MSTR"(ComponentPose=@n15, Alpha=Get Head Down() -> (Head Down:double), Translation="-0.000,-1,0", Rotation="-20.000,0.000,0.000") {
    rotationMode=BMM_Replace
    rotationSpace=BCS_BoneSpace
    scaleMode=BMM_Additive
    scaleSpace=BCS_BoneSpace
    translationMode=BMM_Replace
    translationSpace=BCS_BoneSpace
    alphaBoolBlend {blendOption=Linear}
    alphaScaleBiasClamp {bInterpResult=true, interpSpeedDecreasing=5, interpSpeedIncreasing=7}
    boneToModify {boneName=hair_braid_MSTR}
    rotation=<rot -20 0 0>
    translation {y=-1}
   }
   n17 = "Transform (Modify) Bone - Bone: hair_braid_MSTR"(ComponentPose=@n16, Alpha=Get Head Left() -> (Head Left:double), Translation="0,0,0", Rotation="-5.000,10.000,0.000") {
    rotationMode=BMM_Replace
    rotationSpace=BCS_BoneSpace
    scaleMode=BMM_Additive
    scaleSpace=BCS_BoneSpace
    translationMode=BMM_Replace
    translationSpace=BCS_BoneSpace
    alphaBoolBlend {blendOption=Linear}
    alphaScaleBiasClamp {bInterpResult=true, interpSpeedDecreasing=5, interpSpeedIncreasing=7}
    boneToModify {boneName=hair_braid_MSTR}
    rotation=<rot -5 10 0>
   }
   Component To Local(ComponentPose=@n17)
   n18 = "Transform (Modify) Bone - Bone: hair_braid_01"!(ComponentPose=Local To Component(LocalPose=Use cached pose 'Base'()), Alpha=1, Translation="0,0,0", Rotation="0.000,17.000,0.000") -> (Pose) {
    rotationMode=BMM_Additive
    rotationSpace=BCS_BoneSpace
    scaleMode=BMM_Additive
    scaleSpace=BCS_BoneSpace
    translationMode=BMM_Additive
    translationSpace=BCS_BoneSpace
    alphaBoolBlend {blendOption=Linear}
    boneToModify {boneName=hair_braid_01}
    rotation {yaw=17.0002}
   }
   n19 = Anim Dynamics - Bone: hair_braid_02(ComponentPose=@n18, Alpha=1, ExternalForce="0,0,0") {
    angularDampingOverride=0.9999
    angularSpringConstant=100
    bAngularSpring=true
    bOverrideAngularDamping=true
    bUseSphericalLimits=true
    alphaBoolBlend {blendOption=Linear}
    boundBone {boneName=hair_braid_02}
    componentAppliedLinearAccClamp=<3500 3500 3500>
    componentLinearAccScale=<1.2 1.2 1.2>
    componentLinearVelScale=<1.2 1.2 1.2>
    physicsBodyDefinitions [{
      sphereCollisionRadius=0
      boundBone {boneName=hair_braid_02}
      boxExtents=<20 10 10>
      constraintSetup {angularLimitsMax=<1 10 15>, angularLimitsMin=<-1 -30 -15>, angularTarget=<0 0 0>}
      localJointOffset=<2 0 0>
     }]
    relativeSpaceBone {boneName=dyn_hair_simspace}
    sphericalLimits [{limitRadius=7, drivingBone {boneName=clavicle_l}, sphereLocalOffset=<0 0 5>}]
   }
   n20 = Anim Dynamics - Bone: hair_braid_03(ComponentPose=@n19, Alpha=1, ExternalForce="0,0,0") {
    angularDampingOverride=0.9999
    angularSpringConstant=100
    bAngularSpring=true
    bOverrideAngularDamping=true
    bUseSphericalLimits=true
    alphaBoolBlend {blendOption=Linear}
    boundBone {boneName=hair_braid_03}
    componentAppliedLinearAccClamp=<3500 3500 3500>
    componentLinearAccScale=<1.2 1.2 1.2>
    componentLinearVelScale=<1.2 1.2 1.2>
    physicsBodyDefinitions [{
      sphereCollisionRadius=0
      boundBone {boneName=hair_braid_03}
      boxExtents=<20 10 10>
      constraintSetup {angularLimitsMax=<1 5 15>, angularLimitsMin=<-1 -30 -15>, angularTarget=<0 0 0>}
      localJointOffset=<2 0 0>
     }]
    relativeSpaceBone {boneName=dyn_hair_simspace}
    sphericalLimits [{limitRadius=8, drivingBone {boneName=clavicle_l}, sphereLocalOffset=<0 0 0>}]
   }
   n21 = Anim Dynamics - Bone: hair_braid_04(ComponentPose=@n20, Alpha=1, ExternalForce="0,0,0") {
    angularDampingOverride=0.9999
    angularSpringConstant=100
    bAngularSpring=true
    bOverrideAngularDamping=true
    bUseSphericalLimits=true
    alphaBoolBlend {blendOption=Linear}
    boundBone {boneName=hair_braid_04}
    componentAppliedLinearAccClamp=<3500 3500 3500>
    componentLinearAccScale=<1.2 1.2 1.2>
    componentLinearVelScale=<1.2 1.2 1.2>
    physicsBodyDefinitions [{
      sphereCollisionRadius=0
      boundBone {boneName=hair_braid_04}
      boxExtents=<20 10 10>
      constraintSetup {angularLimitsMax=<1 5 15>, angularLimitsMin=<-1 -30 -15>, angularTarget=<0 0 0>}
      localJointOffset=<2 0 0>
     }]
    relativeSpaceBone {boneName=dyn_hair_simspace}
    sphericalLimits [{limitRadius=8, drivingBone {boneName=clavicle_l}, sphereLocalOffset=<0 0 -3>}]
   }
   n22 = Anim Dynamics - Bone: hair_braid_05(ComponentPose=@n21, Alpha=1, ExternalForce="0,0,0") {
    angularDampingOverride=0.9999
    angularSpringConstant=100
    bAngularSpring=true
    bOverrideAngularDamping=true
    bUseSphericalLimits=true
    alphaBoolBlend {blendOption=Linear}
    boundBone {boneName=hair_braid_05}
    componentAppliedLinearAccClamp=<3500 3500 3500>
    componentLinearAccScale=<1.2 1.2 1.2>
    componentLinearVelScale=<1.2 1.2 1.2>
    physicsBodyDefinitions [{
      sphereCollisionRadius=0
      boundBone {boneName=hair_braid_05}
      boxExtents=<20 10 10>
      constraintSetup {angularLimitsMax=<1 5 15>, angularLimitsMin=<-1 -30 -15>, angularTarget=<0 0 0>}
      localJointOffset=<2 0 0>
     }]
    relativeSpaceBone {boneName=dyn_hair_simspace}
    sphericalLimits [{limitRadius=8, drivingBone {boneName=spine_05}, sphereLocalOffset=<0 0 -8>}]
   }
   n23 = Anim Dynamics - Bone: hair_braid_06(ComponentPose=@n22, Alpha=1, ExternalForce="0,0,0") {
    angularDampingOverride=0.9999
    angularSpringConstant=100
    bAngularSpring=true
    bOverrideAngularDamping=true
    bUseSphericalLimits=true
    alphaBoolBlend {blendOption=Linear}
    boundBone {boneName=hair_braid_06}
    componentAppliedLinearAccClamp=<3500 3500 3500>
    componentLinearAccScale=<1.2 1.2 1.2>
    componentLinearVelScale=<1.2 1.2 1.2>
    physicsBodyDefinitions [{
      sphereCollisionRadius=0
      boundBone {boneName=hair_braid_06}
      boxExtents=<20 10 10>
      constraintSetup {angularLimitsMax=<1 5 15>, angularLimitsMin=<-1 -30 -15>, angularTarget=<0 0 0>}
      localJointOffset=<2 0 0>
     }]
    relativeSpaceBone {boneName=dyn_hair_simspace}
    sphericalLimits [{limitRadius=8, drivingBone {boneName=spine_05}, sphereLocalOffset=<0 0 -8>}]
   }
   n24 = Anim Dynamics - Bone: hair_braid_07(ComponentPose=@n23, Alpha=1, ExternalForce="0,0,0") {
    angularDampingOverride=0.9999
    angularSpringConstant=100
    bAngularSpring=true
    bOverrideAngularDamping=true
    bUseSphericalLimits=true
    alphaBoolBlend {blendOption=Linear}
    boundBone {boneName=hair_braid_07}
    componentAppliedLinearAccClamp=<3500 3500 3500>
    componentLinearAccScale=<1.2 1.2 1.2>
    componentLinearVelScale=<1.2 1.2 1.2>
    physicsBodyDefinitions [{
      sphereCollisionRadius=0
      boundBone {boneName=hair_braid_07}
      boxExtents=<20 10 10>
      constraintSetup {angularLimitsMax=<1 5 15>, angularLimitsMin=<-1 -30 -15>, angularTarget=<0 0 0>}
      localJointOffset=<2 0 0>
     }]
    relativeSpaceBone {boneName=dyn_hair_simspace}
    sphericalLimits [{limitRadius=8, drivingBone {boneName=spine_04}, sphereLocalOffset=<0 0 -8>}]
   }
   n25 = Anim Dynamics - Bone: hair_braid_08(ComponentPose=@n24, Alpha=1, ExternalForce="0,0,0") {
    angularDampingOverride=0.9999
    angularSpringConstant=100
    bAngularSpring=true
    bOverrideAngularDamping=true
    alphaBoolBlend {blendOption=Linear}
    boundBone {boneName=hair_braid_08}
    componentAppliedLinearAccClamp=<3500 3500 3500>
    componentLinearAccScale=<1.2 1.2 1.2>
    componentLinearVelScale=<1.2 1.2 1.2>
    physicsBodyDefinitions [{
      sphereCollisionRadius=0
      boundBone {boneName=hair_braid_08}
      boxExtents=<20 10 10>
      constraintSetup {angularLimitsMax=<1 5 15>, angularLimitsMin=<-1 -30 -15>, angularTarget=<0 0 0>}
      localJointOffset=<2 0 0>
     }]
    relativeSpaceBone {boneName=dyn_hair_simspace}
   }
   n26 = Anim Dynamics - Bone: hair_braid_09(ComponentPose=@n25, Alpha=1, ExternalForce="0,0,0") {
    angularDampingOverride=0.9999
    angularSpringConstant=100
    bAngularSpring=true
    bOverrideAngularDamping=true
    alphaBoolBlend {blendOption=Linear}
    boundBone {boneName=hair_braid_09}
    componentAppliedLinearAccClamp=<3500 3500 3500>
    componentLinearAccScale=<1.2 1.2 1.2>
    componentLinearVelScale=<1.2 1.2 1.2>
    physicsBodyDefinitions [{
      sphereCollisionRadius=0
      boundBone {boneName=hair_braid_09}
      boxExtents=<20 10 10>
      constraintSetup {angularLimitsMax=<1 5 15>, angularLimitsMin=<-1 -30 -15>, angularTarget=<0 0 0>}
      localJointOffset=<2 0 0>
     }]
    relativeSpaceBone {boneName=dyn_hair_simspace}
   }
   n27 = Anim Dynamics - Bone: hair_braid_010(ComponentPose=@n26, Alpha=1, ExternalForce="0,0,0") {
    angularDampingOverride=0.9999
    angularSpringConstant=100
    bAngularSpring=true
    bOverrideAngularDamping=true
    alphaBoolBlend {blendOption=Linear}
    boundBone {boneName=hair_braid_010}
    componentAppliedLinearAccClamp=<3500 3500 3500>
    componentLinearAccScale=<1.2 1.2 1.2>
    componentLinearVelScale=<1.2 1.2 1.2>
    physicsBodyDefinitions [{
      sphereCollisionRadius=0
      boundBone {boneName=hair_braid_010}
      boxExtents=<20 10 10>
      constraintSetup {angularLimitsMax=<1 5 15>, angularLimitsMin=<-1 -30 -15>, angularTarget=<0 0 0>}
      localJointOffset=<2 0 0>
     }]
    relativeSpaceBone {boneName=dyn_hair_simspace}
   }
   n28 = "Constraint - hair_braid_04 (P)"(ComponentPose=@n27, Alpha=0.25, ConstraintWeights_0=1) {
    alphaBoolBlend {blendOption=Linear}
    boneToModify {boneName=hair_braid_04}
    constraintSetup [{targetBone {boneName=clavicle_l}}]
    constraintWeights [1]
   }
   n29 = "Constraint - hair_braid_06 (P)"(ComponentPose=@n28, Alpha=0.15, ConstraintWeights_0=1) {
    alphaBoolBlend {blendOption=Linear}
    boneToModify {boneName=hair_braid_06}
    constraintSetup [{targetBone {boneName=spine_05}}]
    constraintWeights [1]
   }
   Component To Local(ComponentPose=@n29)
   "Constraint - hair_braid_06 (P)"(Alpha=0.5, ConstraintWeights_0=1) {
    alphaBoolBlend {blendOption=Linear}
    boneToModify {boneName=hair_braid_06}
    constraintSetup [{targetBone {boneName=spine_04}}]
    constraintWeights [1]
   }
   "Transform (Modify) Bone - Bone: hair_braid_05"!(ComponentPose, Alpha=1, Translation="0,0,0", Rotation="-30,0,0") -> (Pose) {
    rotationMode=BMM_Additive
    rotationSpace=BCS_BoneSpace
    scaleMode=BMM_Additive
    scaleSpace=BCS_BoneSpace
    translationMode=BMM_Additive
    translationSpace=BCS_BoneSpace
    alphaBoolBlend {blendOption=Linear}
    boneToModify {boneName=hair_braid_05}
    rotation {pitch=-30}
   }
   "Transform (Modify) Bone - Bone: hair_braid_03"(Alpha=1, Translation="0,0,0", Rotation="0,3,0") {
    rotationMode=BMM_Additive
    rotationSpace=BCS_BoneSpace
    scaleMode=BMM_Additive
    scaleSpace=BCS_BoneSpace
    translationMode=BMM_Additive
    translationSpace=BCS_BoneSpace
    alphaBoolBlend {blendOption=Linear}
    boneToModify {boneName=hair_braid_03}
    rotation {yaw=3}
   }
   "Transform (Modify) Bone - Bone: hair_bang_r_03"!(ComponentPose, Alpha=1, Translation="0,0,0", Rotation="15.000,0,0") -> (Pose) {
    rotationMode=BMM_Additive
    rotationSpace=BCS_BoneSpace
    scaleMode=BMM_Additive
    scaleSpace=BCS_BoneSpace
    translationMode=BMM_Additive
    translationSpace=BCS_BoneSpace
    alphaBoolBlend {blendOption=Linear}
    boneToModify {boneName=hair_bang_r_03}
    rotation {pitch=15}
   }
   n30 = "Transform (Modify) Bone - Bone: hair_bang_r_02"(
    ComponentPose=Local To Component(LocalPose=Use cached pose 'SavedPose1779'())
    Alpha=1
    Translation="0,0,0"
    Rotation="4.924,5.037,-8"
    Scale="1,1,1"
   ) {rotationMode=BMM_Additive, alphaBoolBlend {blendOption=Linear}, boneToModify {boneName=hair_bang_r_02}, rotation=<rot 4.924 5.037 -8>}
   n31 = RigidBody(ComponentPose=@n30) {
    bClampLinearTranslationLimitToRefPose=true
    bOverrideWorldGravity=true
    overridePhysicsAsset=&/F_MED_Knight_Cat_FaceAcc_FE_Physics
    simulationSpace=BaseBoneSpace
    alphaBoolBlend {blendOption=Linear}
    baseBoneRef {boneName=neck_01}
    componentAppliedLinearAccClamp=<1000 1000 1000>
    componentLinearAccScale=<0.5 0.5 0.5>
    componentLinearVelScale=<0.5 0.5 0.5>
    overrideWorldGravity {z=-980}
   }
   n32 = "Constraint - hair_braid_05 (PP)"!(
    ComponentPose=Local To Component(LocalPose=Use cached pose 'SavedPose1779'())
    Alpha=0.6
    ConstraintWeights_0=0.7
    ConstraintWeights_1:float=0.3
   ) -> (Pose) {
    alphaBoolBlend {blendOption=Linear}
    alphaScaleBiasClamp {clampMax=0.35}
    boneToModify {boneName=hair_braid_05}
    constraintSetup [{targetBone {boneName=spine_05}}, {targetBone {boneName=clavicle_r}}]
    constraintWeights [1,1]
   }
   n33 = "Constraint - hair_braid_08 (P)"(ComponentPose=@n32, Alpha=0.5, ConstraintWeights_0=1) {
    alphaBoolBlend {blendOption=Linear}
    boneToModify {boneName=hair_braid_08}
    constraintSetup [{targetBone {boneName=spine_04}}]
    constraintWeights [1]
   }
   n34 = RigidBody(ComponentPose=@n33) {
    bClampLinearTranslationLimitToRefPose=true
    bOverrideWorldGravity=true
    overridePhysicsAsset=&F_MED_Knight_Cat_FaceAcc_Physics
    simulationSpace=BaseBoneSpace
    alphaBoolBlend {blendOption=Linear}
    baseBoneRef {boneName=neck_01}
    componentAppliedLinearAccClamp=<1000 1000 1000>
    componentLinearAccScale=<0.5 0.5 0.5>
    componentLinearVelScale=<0.5 0.5 0.5>
    overrideWorldGravity {z=-980}
   }
   Output Pose(Result=Blend Poses by bool(BlendPose_0=Component To Local(ComponentPose=@n31), BlendPose_1=Component To Local(ComponentPose=@n34), bActiveValue=Get bIsInFrontEnd()) {blendType=Linear}) {layerGroup=None}
   n35 = Copy Pose From Mesh(SourceMeshComponent=Get Body_Mesh()) {bCopyCurves=true, bUseAttachedParent=true}
   n36 = Copy Pose From Mesh() {bCopyCurves=true, bUseAttachedParent=true}
   n37 = Anim Dynamics - Bone: hair_tuff_l_02(
    ComponentPose=Local To Component(LocalPose=Layered blend per bone(BasePose=@n36, BlendPoses_0=@n35) {layerSetup [{branchFilters [{boneName=upperarm_r}]}]})
    Alpha=1
    ExternalForce="0,0,0"
   ) {
    angularDampingOverride=0.9999
    angularSpringConstant=500
    bAngularSpring=true
    bOverrideAngularDamping=true
    simulationSpace=BoneRelative
    alphaBoolBlend {blendOption=Linear}
    boundBone {boneName=hair_tuff_l_02}
    componentAppliedLinearAccClamp=<3500 3500 3500>
    componentLinearAccScale=<1.2 1.2 1.2>
    componentLinearVelScale=<1.2 1.2 1.2>
    physicsBodyDefinitions [{
      sphereCollisionRadius=0
      boundBone {boneName=hair_tuff_l_02}
      boxExtents=<10 10 5>
      constraintSetup {angularLimitsMax=<1 5 15>, angularLimitsMin=<-1 -45 -15>, angularTarget=<0 0 0>}
      localJointOffset=<2 0 0>
     }]
    relativeSpaceBone {boneName=dyn_hair_simspace}
   }
   n38 = Anim Dynamics - Bone: hair_tuff_top_02(ComponentPose=@n37, Alpha=1, ExternalForce="0,0,0") {
    angularDampingOverride=0.9999
    angularSpringConstant=500
    bAngularSpring=true
    bOverrideAngularDamping=true
    simulationSpace=BoneRelative
    alphaBoolBlend {blendOption=Linear}
    boundBone {boneName=hair_tuff_top_02}
    componentAppliedLinearAccClamp=<3500 3500 3500>
    componentLinearAccScale=<1.2 1.2 1.2>
    componentLinearVelScale=<1.2 1.2 1.2>
    physicsBodyDefinitions [{
      sphereCollisionRadius=0
      boundBone {boneName=hair_tuff_top_02}
      boxExtents=<15 15 5>
      constraintSetup {angularLimitsMax=<1 15 15>, angularLimitsMin=<-1 -10 -15>, angularTarget=<0 0 0>}
      localJointOffset=<2 0 0>
     }]
    relativeSpaceBone {boneName=dyn_hair_simspace}
   }
   n39 = Anim Dynamics - Bone: hair_strand_l_02(ComponentPose=@n38, Alpha=1, ExternalForce="0,0,0") {
    angularDampingOverride=0.9999
    bGravityOverrideInSimSpace=true
    bOverrideAngularDamping=true
    gravityScale=0.35
    simulationSpace=RootRelative
    alphaBoolBlend {blendOption=Linear}
    boundBone {boneName=hair_strand_l_02}
    componentLinearAccScale {z=0.5}
    componentLinearVelScale {z=0.5}
    physicsBodyDefinitions [{
      sphereCollisionRadius=0
      boundBone {boneName=hair_strand_l_02}
      boxExtents=<5 20 20>
      constraintSetup {angularLimitsMax=<0 10 15>, angularLimitsMin=<0 -5 -5>, angularTarget=<0 0 0>}
      localJointOffset=<5 0 0>
     }]
    relativeSpaceBone {boneName=dyn_hair_simspace}
   }
   n40 = Anim Dynamics - Bone: hair_strand_l_03(ComponentPose=@n39, Alpha=1, ExternalForce="0,0,0") {
    angularDampingOverride=1
    bGravityOverrideInSimSpace=true
    bOverrideAngularDamping=true
    gravityScale=0.35
    simulationSpace=RootRelative
    alphaBoolBlend {blendOption=Linear}
    boundBone {boneName=hair_strand_l_03}
    componentLinearAccScale {z=0.5}
    componentLinearVelScale {z=0.5}
    physicsBodyDefinitions [{
      sphereCollisionRadius=0
      boundBone {boneName=hair_strand_l_03}
      boxExtents=<5 20 20>
      constraintSetup {angularLimitsMax=<0 10 5>, angularLimitsMin=<0 0 -10>, angularTarget=<0 0 0>}
      localJointOffset=<5 0 0>
     }]
    relativeSpaceBone {boneName=dyn_hair_simspace}
   }
   n41 = Anim Dynamics - Bone: hair_strand_c_02(ComponentPose=@n40, Alpha=1, ExternalForce="0,0,0") {
    angularDampingOverride=0.9999
    bGravityOverrideInSimSpace=true
    bOverrideAngularDamping=true
    bUseSphericalLimits=true
    gravityScale=0.35
    simulationSpace=RootRelative
    alphaBoolBlend {blendOption=Linear}
    boundBone {boneName=hair_strand_c_02}
    componentLinearAccScale {z=0.5}
    componentLinearVelScale {z=0.5}
    physicsBodyDefinitions [{
      sphereCollisionRadius=0
      boundBone {boneName=hair_strand_c_02}
      boxExtents=<5 8 8>
      constraintSetup {angularLimitsMax=<0 10 10>, angularLimitsMin=<0 -5 -5>, angularTarget=<0 0 0>}
      localJointOffset=<2 0 0>
     }]
    relativeSpaceBone {boneName=dyn_hair_simspace}
    sphericalLimits [{limitRadius=7, drivingBone {boneName=dyn_hair_simspace}, sphereLocalOffset=<0 2 0>}]
   }
   n42 = Anim Dynamics - Bone: hair_strand_c_03(ComponentPose=@n41, Alpha=1, ExternalForce="0,0,0") {
    angularDampingOverride=1
    bGravityOverrideInSimSpace=true
    bOverrideAngularDamping=true
    gravityScale=0.35
    simulationSpace=RootRelative
    alphaBoolBlend {blendOption=Linear}
    boundBone {boneName=hair_strand_c_03}
    componentLinearAccScale {z=0.5}
    componentLinearVelScale {z=0.5}
    physicsBodyDefinitions [{
      sphereCollisionRadius=0
      boundBone {boneName=hair_strand_c_03}
      boxExtents=<5 10 10>
      constraintSetup {angularLimitsMax=<0 10 10>, angularLimitsMin=<0 0 -2>, angularTarget=<0 0 0>}
      localJointOffset=<-3 0 0>
     }]
    relativeSpaceBone {boneName=dyn_hair_simspace}
   }
   n43 = Anim Dynamics - Bone: hair_strand_c_04(ComponentPose=@n42, Alpha=1, ExternalForce="0,0,0") {
    angularDampingOverride=1
    bGravityOverrideInSimSpace=true
    bOverrideAngularDamping=true
    gravityScale=0.35
    simulationSpace=RootRelative
    alphaBoolBlend {blendOption=Linear}
    boundBone {boneName=hair_strand_c_04}
    componentLinearAccScale {z=0.5}
    componentLinearVelScale {z=0.5}
    physicsBodyDefinitions [{
      sphereCollisionRadius=0
      boundBone {boneName=hair_strand_c_04}
      boxExtents=<5 10 10>
      constraintSetup {angularLimitsMax=<0 5 10>, angularLimitsMin=<0 -5 -2>, angularTarget=<0 0 0>}
      localJointOffset=<-3 0 0>
     }]
    relativeSpaceBone {boneName=dyn_hair_simspace}
   }
   n44 = Anim Dynamics - Bone: hair_strand_c_05(ComponentPose=@n43, Alpha=1, ExternalForce="0,0,0") {
    angularDampingOverride=1
    bGravityOverrideInSimSpace=true
    bOverrideAngularDamping=true
    gravityScale=0.35
    simulationSpace=RootRelative
    alphaBoolBlend {blendOption=Linear}
    boundBone {boneName=hair_strand_c_05}
    componentLinearAccScale {z=0.5}
    componentLinearVelScale {z=0.5}
    physicsBodyDefinitions [{
      sphereCollisionRadius=0
      boundBone {boneName=hair_strand_c_05}
      boxExtents=<5 10 10>
      constraintSetup {angularLimitsMax=<0 0 15>, angularLimitsMin=<0 -5 -2>, angularTarget=<0 0 0>}
      localJointOffset=<5 0 0>
     }]
    relativeSpaceBone {boneName=dyn_hair_simspace}
   }
   n45 = Anim Dynamics - Bone: hair_stand_r_03(ComponentPose=@n44, Alpha=0.35, ExternalForce="0,0,0") {
    angularDampingOverride=1
    bGravityOverrideInSimSpace=true
    bOverrideAngularDamping=true
    bUseSphericalLimits=true
    gravityScale=0.35
    simulationSpace=RootRelative
    alphaBoolBlend {blendOption=Linear}
    boundBone {boneName=hair_stand_r_03}
    componentLinearAccScale=<0.5 0.5 0.5>
    componentLinearVelScale=<0.5 0.5 0.5>
    physicsBodyDefinitions [{
      sphereCollisionRadius=0
      boundBone {boneName=hair_stand_r_03}
      boxExtents=<5 10 10>
      constraintSetup {angularLimitsMax=<0 15 15>, angularLimitsMin=<0 0 -5>, angularTarget=<0 0 0>}
      localJointOffset=<8 0 0>
     }]
    relativeSpaceBone {boneName=dyn_hair_simspace}
    sphericalLimits [{limitRadius=10, drivingBone {boneName=head}, sphereLocalOffset=<0 0 0>}]
   }
   n46 = Anim Dynamics - Bone: hair_stand_r_04(ComponentPose=@n45, Alpha=0.35, ExternalForce="0,0,0") {
    angularDampingOverride=1
    bGravityOverrideInSimSpace=true
    bOverrideAngularDamping=true
    bUseSphericalLimits=true
    gravityScale=0.35
    simulationSpace=RootRelative
    alphaBoolBlend {blendOption=Linear}
    boundBone {boneName=hair_stand_r_04}
    componentLinearAccScale {z=0.5}
    componentLinearVelScale {z=0.5}
    physicsBodyDefinitions [{
      sphereCollisionRadius=0
      boundBone {boneName=hair_stand_r_04}
      boxExtents=<5 8 8>
      constraintSetup {angularLimitsMax=<0 30 10>, angularLimitsMin=<0 0 -10>, angularTarget=<0 0 0>}
      localJointOffset=<5 0 0>
     }]
    relativeSpaceBone {boneName=dyn_hair_simspace}
    sphericalLimits [{limitRadius=15, drivingBone {boneName=dyn_hair_simspace}, sphereLocalOffset=<0 0 0>}]
   }
   Save cached pose 'Base'(Pose=Component To Local(ComponentPose=@n46)) {cacheName=Base}
   Save cached pose 'SavedPose1779'(Pose=Use cached pose 'Base'()) {cacheName=SavedPose1779}
   Get arm_up_l()
  ]}
 EventGraph {nodes [
   Event BlueprintUpdateAnimation()
   > n1 = Sequence!(execute) -> (then_0, then_1, then_2:Exec)
   > n2 = Gravity Override State Function DEPRECATED(CharacterPartAnimInstance=Self-Reference(), GameTimeInSeconds=GetGameTimeInSeconds(), GravityOverrideParams=Get Gravity Override Params())
   > Set GravityOverride(GravityOverride=@n2)
   n3 = BreakTransform(InTransform=GetDeltaTransformFromRefPose(self=GetOwnerSkeletalMeshComponent(), BoneName=upperarm_r, BaseName=spine_05))
   Set arm_up_r(execute=@n1.then_2, arm_up_r:double=MapRangeClamped(Value=BreakRotator(InRot=@n3.Rotation).Yaw, InRangeA=110, InRangeB=130)) -> (Output_Get:double)
   n4 = BreakTransform(InTransform=GetDeltaTransformFromRefPose(self=GetOwnerSkeletalMeshComponent(), BoneName=upperarm_l, BaseName=spine_05))
   Set arm_up_l(execute=@n1.then_1, arm_up_l:double=MapRangeClamped(Value=BreakRotator(InRot=@n4.Rotation).Yaw, InRangeA=45, InRangeB=100)) -> (Output_Get:double)
   n5 = GetOwnerSkeletalMeshComponent()
   n6 = BreakRotator(InRot=BreakTransform(InTransform=GetDeltaTransformFromRefPose(self=@n5, BoneName=head, BaseName=spine_05)).Rotation)
   > Set Head Up(Head Up:double=MapRangeClamped(Value=@n6.Yaw, InRangeA=-3, InRangeB=-40)) -> (Output_Get:double)
   > Set Head Down(Head Down:double=MapRangeClamped(Value=@n6.Yaw, InRangeA=3, InRangeB=20)) -> (Output_Get:double)
   > Set Head Left(Head Left:double=MapRangeClamped(Value=@n6.Pitch, InRangeA=-3, InRangeB=-5)) -> (Output_Get:double)
  ]}
]
variables [
 arm_up_l {type=number}
 arm_up_r {type=number}
 Body_Mesh {type=SkeletalMeshComponent}
 Gravity Override Params {type=GravityOverrideParamsStruct}
 GravityOverride {type=Vector}
 Head Down {type=number}
 Head Left {type=number}
 Head Up {type=number}
]
